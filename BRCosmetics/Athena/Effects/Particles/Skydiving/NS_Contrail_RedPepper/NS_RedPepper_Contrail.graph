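# Niagara skydiving contrail for the RedPepper cosmetic, with fire, orbiting debris and red lightning around the character.
# Fire and FireAir sprites spawn from and follow Soul particles, with parallel _FE emitters reading Soul_FE, and blend cyan to orange by a smoothed system value.
# Debris and lightning meshes attach to sampled skeletal bones or sockets, their spawning scaled by another smoothed value.
# Some lightning particles carry flickering lights.
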NiagaraSystem /BRCosmetics/Athena/Effects/Particles/Skydiving/NS_Contrail_RedPepper/NS_RedPepper_Contrail
dependencies {
 usedDataInterfaces [
  &NiagaraDataInterfaceCurve
  &NiagaraDataInterfaceParticleRead
  &NiagaraDataInterfaceVectorCurve
  &NiagaraDataInterfaceColorCurve
 ]
 usedDynamicInputs [
  &VectorFromCurve
  &FN_FrontEndVector3Modifier
  &TransformVector
  &SetBoolByFloatComparison
  &Divide_Float
  &Abs_Float
  &OneMinusFloat
 ]
 usedModules [
  &self:SystemSpawnScript.SystemScriptSource.SystemScriptGraph.NiagaraNodeAssignment_0.SetVariables_#1
  &SystemState
  /Game/Effects/Niagara/Modules/Skydiving/FN_DivingAttributes/v5/FN_DivingAttributes
  &self:UpdatePreviousRotation
  &self:SystemSpawnScript.SystemScriptSource.SystemScriptGraph.NiagaraNodeAssignment_4.SetVariables_#2
  &self:SystemSpawnScript.SystemScriptSource.SystemScriptGraph.NiagaraNodeAssignment_1.SetVariables_#3
  &FN_Integrator
  &Timeline
  &self:SystemSpawnScript.SystemScriptSource.SystemScriptGraph.NiagaraNodeAssignment_7.SetVariables_#4
  &self:SystemSpawnScript.SystemScriptSource.SystemScriptGraph.NiagaraNodeAssignment_9.SetVariables_#5
 ]
}
emitters {
 Debris {
  emitterUpdateScript [
   &EmitterState()
   &SpawnRate(
    SpawnRate=&Multiply_FloatByInt(
     Float=6
     Integer=&MakeIntFromFloat(Float=$System.InterpolateOverTime002.InterpolateOverTime_Float.Moving Average)
    )
    Use Spawn Probability=false
   )
  ]
  particleSpawnScript [
   /SetVariables(Particles.Random=&RandomRangeFloat(Minimum=0, Maximum=1)) {isSetParameters=true}
   &InitializeParticle(
    Lifetime Max=2.5
    Lifetime Min=1.5
    Lifetime Mode=::NewEnumerator1
    Mesh Scale=&Multiply_VectorByFloat(
     Vector=&RandomRangeVector(Minimum=<0.9 0.9 0.9>, Maximum=<1.5 1.5 1.5>)
     Float=&FloatFromCurve(
      FloatCurve=DataInterface() {propertyValues {Curve {keys [<at 0 0.35 RCIM_Cubic>, {arriveTangent=1.8274, interpMode=RCIM_Cubic, leaveTangent=1.8274, tangentMode=RCTM_User, time=1, value=1}]}}}
      CurveIndex=$Particles.Random
     )
    )
    Mesh Scale Mode=::NewEnumerator4
    UsePositionOffset=false
    Write Material Random=true
   )
   &InitialMeshOrientation(Use Rotation Vector=false)
   /SetVariables(Particles.VisibilityTag=&RandomRangeInt(Minimum=0, Maximum=1)) {isSetParameters=true}
  ]
  particleUpdateScript [
   &ParticleState()
   &ShapeLocation(
    Handle Radius=7
    Large Radius=58
    MaskSpawnGroup=false
    Rotation Coordinate Space=::Simulation
    Shape Primitive=::NewEnumerator3
    Torus Distribution Mode:ENiagaraTorusDistributionMode=::NewEnumerator1
    U Position=&Frac_Float(Float=&Add_Float(
      A=&RandomRangeFloat(Minimum=0, Maximum=1)
      B=&Multiply_Float(A=$Particles.Age, B=&RandomRangeFloat(Minimum=0.05, Maximum=0.15))
     ))
    UseOwnerScale=true
    V Position=&RandomRangeFloat(Minimum=0, Maximum=1)
   )
   &SkeletalMeshLocation(
    Bone / Socket Index=&RandomRangeInt(Minimum=1, Maximum=4)
    Bone Sampling Mode=::NewEnumerator4
    MaskSpawnGroup=false
    Orientation Sampling=::NewEnumerator1
    Skeletal Mesh=$User.SkeletalMesh
    Use Sampled Mesh Attribute as Ribbon ID=false
    Write To Particle Position=false
   )
   &DynamicMaterialParameters(
    Index 0 Param 1=0.02
    Index 0 Param 2=0.25
    Module.Param0WriteEnabled=true
    Module.Param1WriteEnabled=true
    Module.Param2WriteEnabled=false
    Module.Param3WriteEnabled=false
   )
   &ScaleMeshSize(Scale Factor=&VectorFromFloat(Value=&FloatFromCurve(
      FloatCurve=DataInterface() {propertyValues {Curve {keys [
          {arriveTangent=2.3595, interpMode=RCIM_Cubic, leaveTangent=2.3595, tangentMode=RCTM_User, time=0, value=0}
          <at 0.2 1>
          {arriveTangent=-1.25, interpMode=RCIM_Cubic, leaveTangent=-1.25, time=0.75, value=1}
          {arriveTangent=-4.7095, interpMode=RCIM_Cubic, leaveTangent=-4.7095, tangentMode=RCTM_User, time=1, value=0}
         ]}}}
      Scale Curve=$System.InterpolateOverTime002.InterpolateOverTime_Float.Moving Average
     )))
   &UpdateMeshOrientation(
    Rotation Rate=&Lerp_Float(
     A=0.5
     B=0.125
     Alpha=&FloatFromCurve(
      FloatCurve=DataInterface() {propertyValues {Curve {keys [<at 0 0 RCIM_Cubic>, {arriveTangent=3.7701, interpMode=RCIM_Cubic, leaveTangent=3.7701, tangentMode=RCTM_User, time=1, value=1}]}}}
      CurveIndex=$Particles.Random
     )
    )
    Rotation Vector=&RandomVector(Vector Scale=1.25)
   )
   /SetVariables(Particles.Position=&AddVectorToPosition(
     Position=&AddVectorToPosition(
      Position=$Particles.SkeletalMeshLocation.SampledPosition
      Vector=&RotateVectorByQuaternion(Vector To Rotate=<-15 -10 0>, Quaternion=$Particles.SkeletalMeshLocation.SampledRotation)
     )
     Vector=&RotateVectorByQuaternion(
      Vector To Rotate=$Particles.ShapeLocation.ShapeVector
      Quaternion=&MultiplyQuaternion(
       Quaternion A=$Particles.SkeletalMeshLocation.SampledRotation
       Quaternion B=&FindQuatBetweenTwoVectors(From Vector=<1 0 0>, To Vector=<0 0 1>)
      )
     )
    )) {isSetParameters=true}
  ]
  rendererClasses [&NiagaraMeshRendererProperties]
  renderers [
   {rendererClass=&NiagaraMeshRendererProperties, rendererIndex=0}
   {rendererClass=&NiagaraMeshRendererProperties, rendererIndex=1}
  ]
 }
 Fire {
  emitterUpdateScript [
   &EmitterState()
   &SpawnParticlesFromOtherEmitter(
    Attribute Reader=DataInterface() {propertyValues {EmitterBinding {bindingMode=Other, emitterName=Soul}}}
    Spawn Rate=35
    Use Maximum Age=false
    Use Minimum Age=false
   )
  ]
  particleSpawnScript [
   &InitializeParticle(
    Color=<rgba 1 1 1 0.4>
    Lifetime Max=1
    Lifetime Min=0.7
    Lifetime Mode=::NewEnumerator1
    Sprite Rotation Mode=::NewEnumerator2
    Sprite Size Mode=::NewEnumerator1
    Sprite UV Mode=::NewEnumerator1
    Uniform Sprite Size Max=50
    Uniform Sprite Size Min=35
    UsePositionOffset=false
    Write Material Random=true
   )
   &SampleParticlesFromOtherEmitter(Mesh Orientation Sampling:ENiagara_DirectReadSamplingMode=::NewEnumerator0)
   &CameraOffset(Camera Offset Amount=&RandomRangeFloat(Minimum=-8, Maximum=-3))
  ]
  particleUpdateScript [
   &ParticleState()
   &ShapeLocation(
    Additional Quaternion Rotation=false
    Apply To Particle Position=::NewEnumerator1
    MaskSpawnGroup=false
    Non Uniform Scale=<0.8 1 2>
    Offset=<0 0 12>
    Offset Mode=::NewEnumerator0
    Override Seed=false
    Rotation Mode=::NewEnumerator3
    Rotation Quaternion=&MultiplyQuaternion(
     Quaternion A=&self:Get_MeshOrientation_Attribute_By_Index(
      Particle Attribute Reader:NiagaraDataInterfaceParticleRead=DataInterface() {propertyValues {EmitterBinding {bindingMode=Other, emitterName=Soul}}}
      Particle Index:NiagaraInt32=&MakeIntFromFloat(Float=$Particles.SampleParticlesFromOtherEmitter.SampledIndex)
     )
     Quaternion B=&FindQuatBetweenTwoVectors(From Vector=<1 0 0>, To Vector=<1 0 1>)
    )
    Sphere Radius=8
    Transform Order:ENiagara_TransformOrder=::NewEnumerator1
    UseOwnerScale=true
   )
   &InheritSourceMovement(
    Applied Position Scale=&VectorFromFloat(Value=&FloatFromCurve(FloatCurve=DataInterface() {propertyValues {Curve {keys [<at 0 1>,<at 1 0.95>]}}}))
    ApplyForce=false
    ApplyPosition=true
    ApplyPositionFalloff=false
   )
   &SolveForcesAndVelocity(Clamp Velocity=false, Limit Acceleration=false)
   &ScaleSpriteSize(Scale Factor=&Vector2DFromFloat(Value=&FloatFromCurve(FloatCurve=DataInterface() {propertyValues {Curve {keys [
          {arriveTangent=7.5764, interpMode=RCIM_Cubic, leaveTangent=7.5765, tangentMode=RCTM_User, time=0, value=0.1}
          {arriveTangent=0.1045, interpMode=RCIM_Cubic, leaveTangent=0.1045, tangentMode=RCTM_User, time=0.25, value=1}
          {arriveTangent=-0.8944, interpMode=RCIM_Cubic, leaveTangent=-0.8944, tangentMode=RCTM_User, time=1, value=0.8}
         ]}}})))
   &DynamicMaterialParameters(
    Index 0 Param 1=&FloatFromCurve(FloatCurve=DataInterface() {propertyValues {Curve {keys [{arriveTangent=2.7551, interpMode=RCIM_Cubic, leaveTangent=2.7551, tangentMode=RCTM_User, time=0, value=0.5}, <at 0.5 1>]}}})
    Index 0 Param 2=&RandomRangeFloat(Minimum=0.33, Maximum=1)
    Index 0 Param 3=&FloatFromCurve(FloatCurve=DataInterface() {propertyValues {Curve {keys [<at 0 1.25>,<at 1 2>]}}})
    Index 0 Param 4=1
    Module.Param0WriteEnabled=true
    Module.Param1WriteEnabled=true
    Module.Param2WriteEnabled=true
    Module.Param3WriteEnabled=true
   )
   &ScaleColor(
    Scale Mode=::NewEnumerator1
    Scale RGBA=&Vector4FromCurve(Vector4Curve=DataInterface() {propertyValues {
       AlphaCurve {keys [<at 0.75 1>,<at 1 0>]}
       BlueCurve {keys [<at 0 1>,<at 0.1771 15>,<at 0.5441 1.2674>,<at 1 1>]}
       GreenCurve {keys [<at 0 1>,<at 0.1771 15>,<at 0.5441 1.2674>,<at 1 1>]}
       RedCurve {keys [<at 0 1>,<at 0.1771 15>,<at 0.5441 1.2674>,<at 1 1>]}
      }})
    ScaleRGBA=true
   )
   &SubUVAnimation(Start Frame:NiagaraInt32=&RandomRangeInt(Minimum=0, Maximum=20))
   /SetVariables(Particles.Position=/Niagara/DynamicInputs/Lerp/Lerp_Position(
     A:NiagaraPosition=&Add_Vector(
      A=&self:Get_Position_Attribute_By_Index(
       Particle Attribute Reader:NiagaraDataInterfaceParticleRead=DataInterface() {propertyValues {EmitterBinding {bindingMode=Other, emitterName=Soul}}}
       Particle Index:NiagaraInt32=&MakeIntFromFloat(Float=$Particles.SampleParticlesFromOtherEmitter.SampledIndex)
      )
      B=$Particles.ShapeLocation.ShapeVector
     )
     B:NiagaraPosition=&AddVectorToPosition(
      Position=$Particles.Position
      Vector=&RotateVectorByQuaternion(
       Vector To Rotate=<0 0 10>
       Quaternion=&MultiplyQuaternion(
        Quaternion A=&self:Get_MeshOrientation_Attribute_By_Index(
         Particle Attribute Reader:NiagaraDataInterfaceParticleRead=DataInterface() {propertyValues {EmitterBinding {bindingMode=Other, emitterName=Soul}}}
         Particle Index:NiagaraInt32=&MakeIntFromFloat(Float=$Particles.SampleParticlesFromOtherEmitter.SampledIndex)
        )
        Quaternion B=&FindQuatBetweenTwoVectors(From Vector=<1 0 0>, To Vector=<1 0 1>)
       )
      )
     )
     Alpha:NiagaraFloat=&FloatFromCurve(FloatCurve=DataInterface() {propertyValues {Curve {keys [
          {arriveTangent=1.9048, interpMode=RCIM_Cubic, leaveTangent=1.9048, tangentMode=RCTM_User, time=0.1, value=0}
          {arriveTangent=0.5863, interpMode=RCIM_Cubic, leaveTangent=0.5863, tangentMode=RCTM_User, time=1, value=1}
         ]}}})
    )) {isSetParameters=true, module=SetVariables_#6}
   &ScaleColor(
    Scale RGB=&MakeVectorFromLinearColorRGB(Color=&LerpLinearColors(
      StartColor=<rgba 0 1.3565 2 1>
      EndColor=<rgba 2 0.2648 0 1>
      LerpFactor=$System.InterpolateOverTime.InterpolateOverTime_Float.Moving Average
     ))
    ScaleA=true
    ScaleRGB=true
   ) {module=ScaleColor001}
  ]
  rendererClasses [&NiagaraSpriteRendererProperties]
  renderers [{rendererClass=&NiagaraSpriteRendererProperties}]
 }
 FireAir {
  emitterUpdateScript [
   &EmitterState()
   &SpawnParticlesFromOtherEmitter(
    Attribute Reader=DataInterface() {propertyValues {EmitterBinding {bindingMode=Other, emitterName=Soul}}}
    Spawn Rate=8
    Use Maximum Age=false
    Use Minimum Age=false
   )
  ]
  particleSpawnScript [
   /Niagara/Modules/Spawn/Initialization/InitializeParticle(
    Color=<rgba 0.125 0.125 0.125 1>
    Lifetime=&UniformRangedFloat(Minimum=1.5, Maximum=2)
    Module.Write Color=true
    Module.Write Lifetime=true
    Module.Write Mass=true
    Module.Write Material Random=true
    Module.Write Position=false
    Module.Write Scale=false
    Module.Write SpriteRotation=true
    Module.Write SpriteSize=true
    Sprite Rotation=&UniformRangedFloat(Minimum=-35, Maximum=45)
    Sprite Size=<40 40>
   )
   &SampleParticlesFromOtherEmitter(Mesh Orientation Sampling:ENiagara_DirectReadSamplingMode=::NewEnumerator0)
   &CameraOffset(Camera Offset Amount=&RandomRangeFloat(Minimum=0, Maximum=5))
  ]
  particleUpdateScript [
   &ParticleState()
   &ShapeLocation(
    Additional Quaternion Rotation=false
    Apply To Particle Position=::NewEnumerator1
    MaskSpawnGroup=false
    Non Uniform Scale=<0.8 1 2>
    Offset=<0 0 12>
    Offset Mode=::NewEnumerator0
    Override Seed=false
    Rotation Mode=::NewEnumerator3
    Rotation Quaternion=&MultiplyQuaternion(
     Quaternion A=&self:Get_MeshOrientation_Attribute_By_Index(
      Particle Attribute Reader:NiagaraDataInterfaceParticleRead=DataInterface() {propertyValues {EmitterBinding {bindingMode=Other, emitterName=Soul}}}
      Particle Index:NiagaraInt32=&MakeIntFromFloat(Float=$Particles.SampleParticlesFromOtherEmitter.SampledIndex)
     )
     Quaternion B=&FindQuatBetweenTwoVectors(From Vector=<1 0 0>, To Vector=<1 0 1>)
    )
    Sphere Radius=8
    Transform Order:ENiagara_TransformOrder=::NewEnumerator1
    UseOwnerScale=true
   )
   &InheritSourceMovement(
    Applied Position Scale=&VectorFromFloat(Value=&FloatFromCurve(FloatCurve=DataInterface() {propertyValues {Curve {keys [<at 0 1>,<at 1 0.9>]}}}))
    ApplyForce=false
    ApplyPosition=true
    ApplyPositionFalloff=false
   )
   &SolveForcesAndVelocity(Clamp Velocity=false, Limit Acceleration=false)
   &ScaleSpriteSize(Scale Factor=&Vector2DFromFloat(Value=&FloatFromCurve(FloatCurve=DataInterface() {propertyValues {Curve {keys [
          {arriveTangent=3.1239, interpMode=RCIM_Cubic, leaveTangent=3.1239, tangentMode=RCTM_User, time=0, value=0.5}
          {arriveTangent=0.1, interpMode=RCIM_Cubic, leaveTangent=0.1, time=0.1, value=0.975}
          <at 1 0.6 RCIM_Cubic>
         ]}}}))) {module=SpriteSizeScale}
   &SpriteRotationRate(Rotation Rate=&UniformRangedFloat(Minimum=-100, Maximum=100))
   &DynamicMaterialParameters(
    Index 0 Param 1=&RandomRangeFloat(Minimum=0, Maximum=1)
    Index 0 Param 2=&RandomRangeFloat(Minimum=0, Maximum=1)
    Index 0 Param 3=&RandomRangeFloat(Minimum=0.5, Maximum=1)
    Index 0 Param 4=&RandomRangeFloat(Minimum=0.5, Maximum=1)
    Index 1 Param 1=&FloatFromCurve(FloatCurve=DataInterface() {propertyValues {Curve {keys [
         {interpMode=RCIM_Cubic, tangentMode=RCTM_User, time=0, value=1}
         {interpMode=RCIM_Cubic, tangentMode=RCTM_User, time=0.2, value=0.05}
         {interpMode=RCIM_Cubic, tangentMode=RCTM_User, time=0.8, value=0.075}
         {interpMode=RCIM_Cubic, tangentMode=RCTM_User, time=1, value=1.5}
        ]}}})
    Index 1 Param 2=&Multiply_Float(
     A=&RandomRangeFloat(Minimum=-0.25, Maximum=0.25)
     B=&ClampFloat(Float=$User.GlobalScaler, Min=-0.25)
    )
    Index 1 Param 3=&RandomRangeFloat(Minimum=0.6, Maximum=0.8)
    Index 1 Param 4=&RandomRangeFloat(Minimum=5, Maximum=10)
    Index 2 Param 1=&RandomRangeFloat(Minimum=25, Maximum=37.5)
    Index 2 Param 2=&RandomRangeFloat(Minimum=0.4, Maximum=0.5)
    Module.Param0WriteEnabled=true
    Module.Param0WriteEnabled1=true
    Module.Param0WriteEnabled2=true
    Module.Param1WriteEnabled=true
    Module.Param1WriteEnabled1=true
    Module.Param1WriteEnabled2=true
    Module.Param2WriteEnabled=true
    Module.Param2WriteEnabled1=true
    Module.Param2WriteEnabled2=true
    Module.Param3WriteEnabled=true
    Module.Param3WriteEnabled1=true
    Module.Param3WriteEnabled2=true
    Write Parameter Index 1=true
    Write Parameter Index 2=true
   )
   &ScaleColor(
    Scale Alpha=&FloatFromCurve(FloatCurve=DataInterface() {propertyValues {Curve {keys [
         {interpMode=RCIM_Cubic, tangentMode=RCTM_User, time=0, value=0}
         {interpMode=RCIM_Cubic, tangentMode=RCTM_User, time=0.05, value=1}
         {arriveTangent=-2.2003, interpMode=RCIM_Cubic, leaveTangent=-2.2003, tangentMode=RCTM_User, time=1, value=0}
        ]}}})
    Scale RGB=&MakeVectorFromLinearColorRGB(Color=&LerpLinearColors(StartColor=<rgba 0 1.3565 2 1>, EndColor=<rgba 10 1.5 0 1>, LerpFactor=$System.InterpolateOverTime.InterpolateOverTime_Float.Moving Average))
    ScaleA=true
    ScaleRGB=true
   )
   /SetVariables(Particles.Position=&Lerp_Position(
     A:NiagaraPosition=&Add_Vector(
      A=&self:Get_Position_Attribute_By_Index(
       Particle Attribute Reader:NiagaraDataInterfaceParticleRead=DataInterface() {propertyValues {EmitterBinding {bindingMode=Other, emitterName=Soul}}}
       Particle Index:NiagaraInt32=&MakeIntFromFloat(Float=$Particles.SampleParticlesFromOtherEmitter.SampledIndex)
      )
      B=$Particles.ShapeLocation.ShapeVector
     )
     B:NiagaraPosition=&AddVectorToPosition(
      Position=$Particles.Position
      Vector=&RotateVectorByQuaternion(
       Vector To Rotate=<0 0 10>
       Quaternion=&MultiplyQuaternion(
        Quaternion A=&self:Get_MeshOrientation_Attribute_By_Index(
         Particle Attribute Reader:NiagaraDataInterfaceParticleRead=DataInterface() {propertyValues {EmitterBinding {bindingMode=Other, emitterName=Soul}}}
         Particle Index:NiagaraInt32=&MakeIntFromFloat(Float=$Particles.SampleParticlesFromOtherEmitter.SampledIndex)
        )
        Quaternion B=&FindQuatBetweenTwoVectors(From Vector=<1 0 0>, To Vector=<1 0 1>)
       )
      )
     )
     Alpha:NiagaraFloat=&FloatFromCurve(FloatCurve=DataInterface() {propertyValues {Curve {keys [
          {arriveTangent=1.9048, interpMode=RCIM_Cubic, leaveTangent=1.9048, tangentMode=RCTM_User, time=0.1, value=0}
          {arriveTangent=0.5863, interpMode=RCIM_Cubic, leaveTangent=0.5863, tangentMode=RCTM_User, time=1, value=1}
         ]}}})
    )) {isSetParameters=true, module=SetVariables_#6}
  ]
  rendererClasses [&NiagaraSpriteRendererProperties]
  renderers [{rendererClass=&NiagaraSpriteRendererProperties}]
 }
 FireAir_FE {
  emitterUpdateScript [
   &EmitterState()
   &SpawnParticlesFromOtherEmitter(
    Attribute Reader=DataInterface() {propertyValues {EmitterBinding {bindingMode=Other, emitterName=Soul_FE}}}
    Spawn Rate=8
    Use Maximum Age=false
    Use Minimum Age=false
   )
  ]
  particleSpawnScript [
   &InitializeParticle(
    Color=<rgba 0.125 0.125 0.125 1>
    Lifetime=&UniformRangedFloat(Minimum=1.5, Maximum=2)
    Module.Write Color=true
    Module.Write Lifetime=true
    Module.Write Mass=true
    Module.Write Material Random=true
    Module.Write Position=false
    Module.Write Scale=false
    Module.Write SpriteRotation=true
    Module.Write SpriteSize=true
    Sprite Rotation=&UniformRangedFloat(Minimum=-35, Maximum=45)
    Sprite Size=<40 40>
   )
   &SampleParticlesFromOtherEmitter(Mesh Orientation Sampling:ENiagara_DirectReadSamplingMode=::NewEnumerator0)
   &CameraOffset(Camera Offset Amount=&RandomRangeFloat(Minimum=0, Maximum=5))
  ]
  particleUpdateScript [
   &ParticleState()
   &ShapeLocation(
    Additional Quaternion Rotation=false
    Apply To Particle Position=::NewEnumerator1
    MaskSpawnGroup=false
    Non Uniform Scale=<0.8 1 2>
    Offset=<0 0 12>
    Offset Mode=::NewEnumerator0
    Override Seed=false
    Rotation Mode=::NewEnumerator3
    Rotation Quaternion=&MultiplyQuaternion(
     Quaternion A=&self:Get_MeshOrientation_Attribute_By_Index(
      Particle Attribute Reader:NiagaraDataInterfaceParticleRead=DataInterface() {propertyValues {EmitterBinding {bindingMode=Other, emitterName=Soul_FE}}}
      Particle Index:NiagaraInt32=&MakeIntFromFloat(Float=$Particles.SampleParticlesFromOtherEmitter.SampledIndex)
     )
     Quaternion B=&FindQuatBetweenTwoVectors(From Vector=<1 0 0>, To Vector=<1 0 1>)
    )
    Sphere Radius=8
    Transform Order:ENiagara_TransformOrder=::NewEnumerator1
    UseOwnerScale=true
   )
   &InheritSourceMovement(
    Applied Position Scale=&VectorFromFloat(Value=&FloatFromCurve(FloatCurve=DataInterface() {propertyValues {Curve {keys [<at 0 1>,<at 1 0.9>]}}}))
    ApplyForce=false
    ApplyPosition=true
    ApplyPositionFalloff=false
   )
   &SolveForcesAndVelocity(Clamp Velocity=false, Limit Acceleration=false)
   &ScaleSpriteSize(Scale Factor=&Vector2DFromFloat(Value=&FloatFromCurve(FloatCurve=DataInterface() {propertyValues {Curve {keys [
          {arriveTangent=3.1239, interpMode=RCIM_Cubic, leaveTangent=3.1239, tangentMode=RCTM_User, time=0, value=0.5}
          {arriveTangent=0.1, interpMode=RCIM_Cubic, leaveTangent=0.1, time=0.1, value=0.975}
          <at 1 0.6 RCIM_Cubic>
         ]}}}))) {module=SpriteSizeScale}
   &SpriteRotationRate(Rotation Rate=&UniformRangedFloat(Minimum=-100, Maximum=100))
   &DynamicMaterialParameters(
    Index 0 Param 1=&RandomRangeFloat(Minimum=0, Maximum=1)
    Index 0 Param 2=&RandomRangeFloat(Minimum=0, Maximum=1)
    Index 0 Param 3=&RandomRangeFloat(Minimum=0.5, Maximum=1)
    Index 0 Param 4=&RandomRangeFloat(Minimum=0.5, Maximum=1)
    Index 1 Param 1=&FloatFromCurve(FloatCurve=DataInterface() {propertyValues {Curve {keys [
         {interpMode=RCIM_Cubic, tangentMode=RCTM_User, time=0, value=1}
         {interpMode=RCIM_Cubic, tangentMode=RCTM_User, time=0.2, value=0.05}
         {interpMode=RCIM_Cubic, tangentMode=RCTM_User, time=0.8, value=0.075}
         {interpMode=RCIM_Cubic, tangentMode=RCTM_User, time=1, value=1.5}
        ]}}})
    Index 1 Param 2=&Multiply_Float(
     A=&RandomRangeFloat(Minimum=-0.25, Maximum=0.25)
     B=&ClampFloat(Float=$User.GlobalScaler, Min=-0.25)
    )
    Index 1 Param 3=&RandomRangeFloat(Minimum=0.6, Maximum=0.8)
    Index 1 Param 4=&RandomRangeFloat(Minimum=5, Maximum=10)
    Index 2 Param 1=&RandomRangeFloat(Minimum=25, Maximum=37.5)
    Index 2 Param 2=&RandomRangeFloat(Minimum=0.4, Maximum=0.5)
    Module.Param0WriteEnabled=true
    Module.Param0WriteEnabled1=true
    Module.Param0WriteEnabled2=true
    Module.Param1WriteEnabled=true
    Module.Param1WriteEnabled1=true
    Module.Param1WriteEnabled2=true
    Module.Param2WriteEnabled=true
    Module.Param2WriteEnabled1=true
    Module.Param2WriteEnabled2=true
    Module.Param3WriteEnabled=true
    Module.Param3WriteEnabled1=true
    Module.Param3WriteEnabled2=true
    Write Parameter Index 1=true
    Write Parameter Index 2=true
   )
   &ScaleColor(
    Scale Alpha=&FloatFromCurve(FloatCurve=DataInterface() {propertyValues {Curve {keys [
         {interpMode=RCIM_Cubic, tangentMode=RCTM_User, time=0, value=0}
         {interpMode=RCIM_Cubic, tangentMode=RCTM_User, time=0.05, value=1}
         {arriveTangent=-2.2003, interpMode=RCIM_Cubic, leaveTangent=-2.2003, tangentMode=RCTM_User, time=1, value=0}
        ]}}})
    Scale RGB=&MakeVectorFromLinearColorRGB(Color=&LerpLinearColors(StartColor=<rgba 0 1.3565 2 1>, EndColor=<rgba 10 1.5 0 1>, LerpFactor=$System.InterpolateOverTime.InterpolateOverTime_Float.Moving Average))
    ScaleA=true
    ScaleRGB=true
   )
   /SetVariables(Particles.Position=&Lerp_Position(
     A:NiagaraPosition=&Add_Vector(
      A=&self:Get_Position_Attribute_By_Index(
       Particle Attribute Reader:NiagaraDataInterfaceParticleRead=DataInterface() {propertyValues {EmitterBinding {bindingMode=Other, emitterName=Soul_FE}}}
       Particle Index:NiagaraInt32=&MakeIntFromFloat(Float=$Particles.SampleParticlesFromOtherEmitter.SampledIndex)
      )
      B=$Particles.ShapeLocation.ShapeVector
     )
     B:NiagaraPosition=&AddVectorToPosition(
      Position=$Particles.Position
      Vector=&RotateVectorByQuaternion(
       Vector To Rotate=&Lerp_Vector(A=<0 0 10>, B=<0 0 25>, Alpha=$System.InterpolateOverTime.InterpolateOverTime_Float.Moving Average)
       Quaternion=&MultiplyQuaternion(
        Quaternion A=&self:Get_MeshOrientation_Attribute_By_Index(
         Particle Attribute Reader:NiagaraDataInterfaceParticleRead=DataInterface() {propertyValues {EmitterBinding {bindingMode=Other, emitterName=Soul_FE}}}
         Particle Index:NiagaraInt32=&MakeIntFromFloat(Float=$Particles.SampleParticlesFromOtherEmitter.SampledIndex)
        )
        Quaternion B=&FindQuatBetweenTwoVectors(From Vector=<1 0 0>, To Vector=<0.5 0 1>)
       )
      )
     )
     Alpha:NiagaraFloat=&FloatFromCurve(FloatCurve=DataInterface() {propertyValues {Curve {keys [
          {arriveTangent=1.9048, interpMode=RCIM_Cubic, leaveTangent=1.9048, tangentMode=RCTM_User, time=0.1, value=0}
          {arriveTangent=0.5863, interpMode=RCIM_Cubic, leaveTangent=0.5863, tangentMode=RCTM_User, time=1, value=1}
         ]}}})
    )) {isSetParameters=true, module=SetVariables_#6}
  ]
  rendererClasses [&NiagaraSpriteRendererProperties]
  renderers [{rendererClass=&NiagaraSpriteRendererProperties}]
 }
 Fire_FE {
  emitterUpdateScript [
   &EmitterState()
   &SpawnParticlesFromOtherEmitter(
    Attribute Reader=DataInterface() {propertyValues {EmitterBinding {bindingMode=Other, emitterName=Soul_FE}}}
    Spawn Rate=35
    Use Maximum Age=false
    Use Minimum Age=false
   )
  ]
  particleSpawnScript [
   /Niagara/Modules/Spawn/Initialization/V2/InitializeParticle(
    Color=<rgba 1 1 1 0.4>
    Lifetime Max=1
    Lifetime Min=0.7
    Lifetime Mode=::NewEnumerator1
    Sprite Rotation Mode=::NewEnumerator2
    Sprite Size Mode=::NewEnumerator1
    Sprite UV Mode=::NewEnumerator1
    Uniform Sprite Size Max=50
    Uniform Sprite Size Min=35
    UsePositionOffset=false
    Write Material Random=true
   )
   &SampleParticlesFromOtherEmitter(Mesh Orientation Sampling:ENiagara_DirectReadSamplingMode=::NewEnumerator0)
   &CameraOffset(Camera Offset Amount=&RandomRangeFloat(Minimum=-8, Maximum=-3))
  ]
  particleUpdateScript [
   &ParticleState()
   &ShapeLocation(
    Additional Quaternion Rotation=false
    Apply To Particle Position=::NewEnumerator1
    MaskSpawnGroup=false
    Non Uniform Scale=<0.8 1 2>
    Offset=<0 0 12>
    Offset Mode=::NewEnumerator0
    Override Seed=false
    Rotation Mode=::NewEnumerator3
    Rotation Quaternion=&MultiplyQuaternion(
     Quaternion A=&self:Get_MeshOrientation_Attribute_By_Index(
      Particle Attribute Reader:NiagaraDataInterfaceParticleRead=DataInterface() {propertyValues {EmitterBinding {bindingMode=Other, emitterName=Soul_FE}}}
      Particle Index:NiagaraInt32=&MakeIntFromFloat(Float=$Particles.SampleParticlesFromOtherEmitter.SampledIndex)
     )
     Quaternion B=&FindQuatBetweenTwoVectors(From Vector=<1 0 0>, To Vector=<1 0 1>)
    )
    Sphere Radius=8
    Transform Order:ENiagara_TransformOrder=::NewEnumerator1
    UseOwnerScale=true
   )
   &InheritSourceMovement(
    Applied Position Scale=&VectorFromFloat(Value=&FloatFromCurve(FloatCurve=DataInterface() {propertyValues {Curve {keys [<at 0 1>,<at 1 0.95>]}}}))
    ApplyForce=false
    ApplyPosition=true
    ApplyPositionFalloff=false
   )
   &SolveForcesAndVelocity(Clamp Velocity=false, Limit Acceleration=false)
   &ScaleSpriteSize(Scale Factor=&Vector2DFromFloat(Value=&FloatFromCurve(FloatCurve=DataInterface() {propertyValues {Curve {keys [
          {arriveTangent=7.5764, interpMode=RCIM_Cubic, leaveTangent=7.5765, tangentMode=RCTM_User, time=0, value=0.1}
          {arriveTangent=0.1045, interpMode=RCIM_Cubic, leaveTangent=0.1045, tangentMode=RCTM_User, time=0.25, value=1}
          {arriveTangent=-0.8944, interpMode=RCIM_Cubic, leaveTangent=-0.8944, tangentMode=RCTM_User, time=1, value=0.8}
         ]}}})))
   &DynamicMaterialParameters(
    Index 0 Param 1=&FloatFromCurve(FloatCurve=DataInterface() {propertyValues {Curve {keys [{arriveTangent=2.7551, interpMode=RCIM_Cubic, leaveTangent=2.7551, tangentMode=RCTM_User, time=0, value=0.5}, <at 0.5 1>]}}})
    Index 0 Param 2=&RandomRangeFloat(Minimum=0.33, Maximum=1)
    Index 0 Param 3=&FloatFromCurve(FloatCurve=DataInterface() {propertyValues {Curve {keys [<at 0 1.25>,<at 1 2>]}}})
    Index 0 Param 4=1
    Module.Param0WriteEnabled=true
    Module.Param1WriteEnabled=true
    Module.Param2WriteEnabled=true
    Module.Param3WriteEnabled=true
   )
   &ScaleColor(
    Scale Mode=::NewEnumerator1
    Scale RGBA=&Vector4FromCurve(Vector4Curve=DataInterface() {propertyValues {
       AlphaCurve {keys [<at 0.75 1>,<at 1 0>]}
       BlueCurve {keys [<at 0 1>,<at 0.1771 15>,<at 0.5441 1.2674>,<at 1 1>]}
       GreenCurve {keys [<at 0 1>,<at 0.1771 15>,<at 0.5441 1.2674>,<at 1 1>]}
       RedCurve {keys [<at 0 1>,<at 0.1771 15>,<at 0.5441 1.2674>,<at 1 1>]}
      }})
    ScaleRGBA=true
   )
   &SubUVAnimation(Start Frame:NiagaraInt32=&RandomRangeInt(Minimum=0, Maximum=20))
   /SetVariables(Particles.Position=&Lerp_Position(
     A:NiagaraPosition=&Add_Vector(
      A=&self:Get_Position_Attribute_By_Index(
       Particle Attribute Reader:NiagaraDataInterfaceParticleRead=DataInterface() {propertyValues {EmitterBinding {bindingMode=Other, emitterName=Soul_FE}}}
       Particle Index:NiagaraInt32=&MakeIntFromFloat(Float=$Particles.SampleParticlesFromOtherEmitter.SampledIndex)
      )
      B=$Particles.ShapeLocation.ShapeVector
     )
     B:NiagaraPosition=&AddVectorToPosition(
      Position=$Particles.Position
      Vector=&RotateVectorByQuaternion(
       Vector To Rotate=&Lerp_Vector(A=<0 0 15>, B=<0 0 40>, Alpha=$System.InterpolateOverTime.InterpolateOverTime_Float.Moving Average)
       Quaternion=&MultiplyQuaternion(
        Quaternion A=&self:Get_MeshOrientation_Attribute_By_Index(
         Particle Attribute Reader:NiagaraDataInterfaceParticleRead=DataInterface() {propertyValues {EmitterBinding {bindingMode=Other, emitterName=Soul_FE}}}
         Particle Index:NiagaraInt32=&MakeIntFromFloat(Float=$Particles.SampleParticlesFromOtherEmitter.SampledIndex)
        )
        Quaternion B=&FindQuatBetweenTwoVectors(From Vector=<1 0 0>, To Vector=<0.5 0 1>)
       )
      )
     )
     Alpha:NiagaraFloat=&FloatFromCurve(FloatCurve=DataInterface() {propertyValues {Curve {keys [{tangentMode=RCTM_User, time=0, value=0}, {tangentMode=RCTM_User, time=1, value=1}]}}})
    )) {isSetParameters=true, module=SetVariables_#6}
   &ScaleColor(
    Scale RGB=&MakeVectorFromLinearColorRGB(Color=&LerpLinearColors(
      StartColor=<rgba 0 1.3565 2 1>
      EndColor=<rgba 2 0.2648 0 1>
      LerpFactor=$System.InterpolateOverTime.InterpolateOverTime_Float.Moving Average
     ))
    ScaleA=true
    ScaleRGB=true
   ) {module=ScaleColor001}
  ]
  rendererClasses [&NiagaraSpriteRendererProperties]
  renderers [{rendererClass=&NiagaraSpriteRendererProperties}]
 }
 Lightning {
  emitterUpdateScript [
   &EmitterState(
    Life Cycle Mode=::NewEnumerator1
    Loop Duration=&RandomRangeFloat(Minimum=0.1, Maximum=0.35)
    Recalculate Duration Each Loop=true
    UseLoopDelay=false
   )
   &SpawnBurst_Instantaneous(
    Spawn Count=&Multiply_Int(
     A=&RandomRangeInt(Minimum=2, Maximum=5)
     B=&MakeIntFromFloat(Float=$System.InterpolateOverTime002.InterpolateOverTime_Float.Moving Average)
    )
    Use Spawn Probability=false
    UseLoopCountLimit=false
   )
   &SpawnRate(
    SpawnRate=&Multiply_FloatByInt(
     Float=2
     Integer=&MakeIntFromFloat(Float=$System.InterpolateOverTime002.InterpolateOverTime_Float.Moving Average)
    )
    Use Spawn Probability=false
   ) {enabled=false}
  ]
  particleSpawnScript [
   &InitializeParticle(
    Color Maximum=<rgba 14 0.4375 0.4375 1>
    Color Minimum=<rgba 6 0.0625 0.0625 1>
    Color Mode=::NewEnumerator2
    Lifetime Max=0.4
    Lifetime Min=0.2
    Lifetime Mode=::NewEnumerator1
    Mesh Scale Max=<1.2 1.2 1.2>
    Mesh Scale Min=<0.8 0.8 0.8>
    Mesh Scale Mode=::NewEnumerator2
    UsePositionOffset=false
    Write Material Random=true
   )
   &InitialMeshOrientation(
    Mesh Axis To Orient=<0 0 1>
    Mesh Orientation Mode=::NewEnumerator2
    Rotation=&RandomRangeVector(Minimum=<-0.07 -0.07 -1>, Maximum=<0.07 0.07 1>)
    Use Rotation Vector=true
   )
   &ShapeLocation(
    Handle Radius=10
    Large Radius=35
    MaskSpawnGroup=false
    Override Seed=false
    Rotation Coordinate Space=::Simulation
    Shape Primitive=::NewEnumerator3
    UseOwnerScale=true
   )
   /SetVariables(
    Particles.LightningColor:LinearColor=&RandomRangeLinearColor(Minimum=<rgba 18 0.4445 0.371 1>, Maximum=<rgba 20 0 0 1>)
    Particles.LightningScale:Vector3f=&VectorFromFloat(Value=&RandomRangeFloat(Minimum=0.4, Maximum=0.5))
   ) {isSetParameters=true, module=SetVariables_#7}
   /SetVariables(Particles.VisibilityTag=&RandomRangeInt(Minimum=0, Maximum=2)) {isSetParameters=true, module=SetVariables_#8}
  ]
  particleUpdateScript [
   &ParticleState()
   &SkeletalMeshLocation(
    Bone / Socket Index=&RandomRangeInt(Minimum=1, Maximum=4)
    Bone Sampling Mode=::NewEnumerator4
    MaskSpawnGroup=false
    Sampled Position Offset=<-15 -10 0>
    Skeletal Mesh=$User.SkeletalMesh
   )
   &DynamicMaterialParameters(
    Index 0 Param 1=&RandomRangeFloat(Minimum=2.5, Maximum=3)
    Index 0 Param 2=0.25
    Index 0 Param 3=$Particles.NormalizedAge
    Index 0 Param 4=&RandomRangeFloat(Minimum=0.85, Maximum=1)
    Index 1 Param 1=&RandomRangeFloat(Minimum=4, Maximum=5)
    Index 1 Param 2=0.5
    Index 1 Param 3=$Particles.NormalizedAge
    Index 1 Param 4=&RandomRangeFloat(Minimum=2, Maximum=2.4)
    Module.Param0WriteEnabled=true
    Module.Param0WriteEnabled1=true
    Module.Param1WriteEnabled=true
    Module.Param1WriteEnabled1=true
    Module.Param2WriteEnabled=true
    Module.Param2WriteEnabled1=true
    Module.Param3WriteEnabled=true
    Module.Param3WriteEnabled1=true
    Write Parameter Index 1=true
   )
   &UpdateMeshOrientation(Rotation Rate=0.25, Rotation Vector=<0 0 1>)
   &FN_Light_Attributes(
    Color=<rgba 10 0.1562 0.1562 1>
    Create Color Attribute=true
    Create Exponent Attribute=true
    Create Position Attribute=true
    Create Radius Attribute=true
    Enable Particle Light Enable=true
    Exponent=2
    Light Enabled:NiagaraBool=&SetBoolByIntComparison(
     A=&ModuloInt(A=&ReturnParticleIDAttributes(), B=3)
     B=1
     Comparison Type=::NewEnumerator4
    )
    Position=&AddVectorToPosition(
     Position=&AddVectorToPosition(
      Position=$Particles.SkeletalMeshLocation.SampledPosition
      Vector=&RotateVectorByQuaternion(Vector To Rotate=<-15 -10 0>, Quaternion=$Particles.SkeletalMeshLocation.SampledRotation)
     )
     Vector=&RotateVectorByQuaternion(
      Vector To Rotate=$Particles.ShapeLocation.ShapeVector
      Quaternion=&MultiplyQuaternion(
       Quaternion A=$Particles.SkeletalMeshLocation.SampledRotation
       Quaternion B=&FindQuatBetweenTwoVectors(From Vector=<1 0 0>, To Vector=<0 0 1>)
      )
     )
    )
    Radius=&FloatFromCurve(
     FloatCurve=DataInterface() {propertyValues {Curve {keys [
         <at 0 1 RCIM_Cubic>
         {arriveTangent=1.794, interpMode=RCIM_Cubic, leaveTangent=1.794, time=0.0746, value=0.7914}
         {arriveTangent=-0.0227, interpMode=RCIM_Cubic, leaveTangent=-0.0227, tangentMode=RCTM_User, time=0.121, value=1.2171}
         {arriveTangent=-5.5976, interpMode=RCIM_Cubic, leaveTangent=-5.5976, time=0.1377, value=0.3589}
         {arriveTangent=5.9152, interpMode=RCIM_Cubic, leaveTangent=5.9152, time=0.1738, value=0.9213}
         {arriveTangent=3.6344, interpMode=RCIM_Cubic, leaveTangent=3.6344, time=0.2093, value=0.7825}
         {arriveTangent=-4.8832, interpMode=RCIM_Cubic, leaveTangent=-4.8832, time=0.2396, value=1.1602}
         {arriveTangent=-6.1936, interpMode=RCIM_Cubic, leaveTangent=-6.1936, time=0.254, value=0.5643}
         {arriveTangent=-3.4254, interpMode=RCIM_Cubic, leaveTangent=-3.4254, time=0.2791, value=0.9154}
         {arriveTangent=1.2156, interpMode=RCIM_Cubic, leaveTangent=1.2156, time=0.3019, value=0.4002}
         {arriveTangent=5.6358, interpMode=RCIM_Cubic, leaveTangent=5.6358, time=0.3406, value=0.9901}
         {arriveTangent=6.3907, interpMode=RCIM_Cubic, leaveTangent=6.3907, time=0.3796, value=0.8384}
         {arriveTangent=-3.0158, interpMode=RCIM_Cubic, leaveTangent=-3.0158, time=0.4112, value=1.4411}
         {arriveTangent=-10.4214, interpMode=RCIM_Cubic, leaveTangent=-10.4214, time=0.4339, value=0.6747}
         {arriveTangent=-3.4821, interpMode=RCIM_Cubic, leaveTangent=-3.4821, time=0.4676, value=0.8529}
         {arriveTangent=-1.4741, interpMode=RCIM_Cubic, leaveTangent=-1.4741, time=0.4906, value=0.4775}
         {arriveTangent=-2.6963, interpMode=RCIM_Cubic, leaveTangent=-2.6963, time=0.5283, value=0.7634}
         {arriveTangent=2.8896, interpMode=RCIM_Cubic, leaveTangent=2.8896, time=0.5517, value=0.3126}
         {arriveTangent=6.2844, interpMode=RCIM_Cubic, leaveTangent=6.2844, time=0.5824, value=0.9196}
         {arriveTangent=1.4786, interpMode=RCIM_Cubic, leaveTangent=1.4786, time=0.5976, value=0.6006}
         {arriveTangent=4.8152, interpMode=RCIM_Cubic, leaveTangent=4.8152, time=0.6363, value=0.9993}
         {arriveTangent=2.8106, interpMode=RCIM_Cubic, leaveTangent=2.8106, time=0.6464, value=0.8356}
         {arriveTangent=3.2181, interpMode=RCIM_Cubic, leaveTangent=3.2181, time=0.6964, value=1.1681}
         {arriveTangent=1.8672, interpMode=RCIM_Cubic, leaveTangent=1.8672, time=0.7177, value=1.065}
         {arriveTangent=-12.2147, interpMode=RCIM_Cubic, leaveTangent=-12.2147, time=0.7385, value=1.2469}
         {arriveTangent=-0.0928, interpMode=RCIM_Cubic, leaveTangent=-0.0928, tangentMode=RCTM_User, time=0.7564, value=0.5916}
         {arriveTangent=3.085, interpMode=RCIM_Cubic, leaveTangent=3.085, time=0.7931, value=1.02}
         {arriveTangent=2.1222, interpMode=RCIM_Cubic, leaveTangent=2.1222, time=0.8092, value=0.7545}
         {arriveTangent=-0.424, interpMode=RCIM_Cubic, leaveTangent=-0.424, time=0.8924, value=1.2306}
         {arriveTangent=-8.511, interpMode=RCIM_Cubic, leaveTangent=-8.511, time=0.911, value=0.7114}
         {arriveTangent=3.2422, interpMode=RCIM_Cubic, leaveTangent=3.2422, time=0.9278, value=0.9291}
         {interpMode=RCIM_Cubic, leaveTangent=0.0463, time=1, value=1}
        ]}}}
     Scale Curve=50
    )
    Scale Color=&FloatFromCurve(
     FloatCurve=DataInterface() {propertyValues {Curve {keys [
         <at 0 1 RCIM_Cubic>
         {arriveTangent=1.794, interpMode=RCIM_Cubic, leaveTangent=1.794, time=0.0746, value=0.7914}
         {arriveTangent=-0.0227, interpMode=RCIM_Cubic, leaveTangent=-0.0227, tangentMode=RCTM_User, time=0.121, value=1.2171}
         {arriveTangent=-5.5976, interpMode=RCIM_Cubic, leaveTangent=-5.5976, time=0.1377, value=0.3589}
         {arriveTangent=5.9152, interpMode=RCIM_Cubic, leaveTangent=5.9152, time=0.1738, value=0.9213}
         {arriveTangent=3.6344, interpMode=RCIM_Cubic, leaveTangent=3.6344, time=0.2093, value=0.7825}
         {arriveTangent=-4.8832, interpMode=RCIM_Cubic, leaveTangent=-4.8832, time=0.2396, value=1.1602}
         {arriveTangent=-6.1936, interpMode=RCIM_Cubic, leaveTangent=-6.1936, time=0.254, value=0.5643}
         {arriveTangent=-3.4254, interpMode=RCIM_Cubic, leaveTangent=-3.4254, time=0.2791, value=0.9154}
         {arriveTangent=1.2156, interpMode=RCIM_Cubic, leaveTangent=1.2156, time=0.3019, value=0.4002}
         {arriveTangent=5.6358, interpMode=RCIM_Cubic, leaveTangent=5.6358, time=0.3406, value=0.9901}
         {arriveTangent=6.3907, interpMode=RCIM_Cubic, leaveTangent=6.3907, time=0.3796, value=0.8384}
         {arriveTangent=-3.0158, interpMode=RCIM_Cubic, leaveTangent=-3.0158, time=0.4112, value=1.4411}
         {arriveTangent=-10.4214, interpMode=RCIM_Cubic, leaveTangent=-10.4214, time=0.4339, value=0.6747}
         {arriveTangent=-3.4821, interpMode=RCIM_Cubic, leaveTangent=-3.4821, time=0.4676, value=0.8529}
         {arriveTangent=-1.4741, interpMode=RCIM_Cubic, leaveTangent=-1.4741, time=0.4906, value=0.4775}
         {arriveTangent=-2.6963, interpMode=RCIM_Cubic, leaveTangent=-2.6963, time=0.5283, value=0.7634}
         {arriveTangent=2.8896, interpMode=RCIM_Cubic, leaveTangent=2.8896, time=0.5517, value=0.3126}
         {arriveTangent=6.2844, interpMode=RCIM_Cubic, leaveTangent=6.2844, time=0.5824, value=0.9196}
         {arriveTangent=1.4786, interpMode=RCIM_Cubic, leaveTangent=1.4786, time=0.5976, value=0.6006}
         {arriveTangent=4.8152, interpMode=RCIM_Cubic, leaveTangent=4.8152, time=0.6363, value=0.9993}
         {arriveTangent=2.8106, interpMode=RCIM_Cubic, leaveTangent=2.8106, time=0.6464, value=0.8356}
         {arriveTangent=3.2181, interpMode=RCIM_Cubic, leaveTangent=3.2181, time=0.6964, value=1.1681}
         {arriveTangent=1.8672, interpMode=RCIM_Cubic, leaveTangent=1.8672, time=0.7177, value=1.065}
         {arriveTangent=-12.2147, interpMode=RCIM_Cubic, leaveTangent=-12.2147, time=0.7385, value=1.2469}
         {arriveTangent=-0.0928, interpMode=RCIM_Cubic, leaveTangent=-0.0928, tangentMode=RCTM_User, time=0.7564, value=0.5916}
         {arriveTangent=3.085, interpMode=RCIM_Cubic, leaveTangent=3.085, time=0.7931, value=1.02}
         {arriveTangent=2.1222, interpMode=RCIM_Cubic, leaveTangent=2.1222, time=0.8092, value=0.7545}
         {arriveTangent=-0.424, interpMode=RCIM_Cubic, leaveTangent=-0.424, time=0.8924, value=1.2306}
         {arriveTangent=-8.511, interpMode=RCIM_Cubic, leaveTangent=-8.511, time=0.911, value=0.7114}
         {arriveTangent=3.2422, interpMode=RCIM_Cubic, leaveTangent=3.2422, time=0.9278, value=0.9291}
         {interpMode=RCIM_Cubic, leaveTangent=0.0463, time=1, value=1}
        ]}}}
     Scale Curve=3
    )
   )
  ]
  rendererClasses [&NiagaraMeshRendererProperties, &NiagaraLightRendererProperties]
  renderers [
   {rendererClass=&NiagaraMeshRendererProperties, rendererIndex=0}
   {rendererClass=&NiagaraLightRendererProperties, rendererIndex=1}
   {rendererClass=&NiagaraMeshRendererProperties, rendererIndex=2}
   {rendererClass=&NiagaraMeshRendererProperties, rendererIndex=3}
  ]
 }
 Lightning_Mobile {
  emitterUpdateScript [
   &EmitterState(
    Life Cycle Mode=::NewEnumerator1
    Loop Duration=&RandomRangeFloat(Minimum=0.1, Maximum=0.35)
    Recalculate Duration Each Loop=true
    UseLoopDelay=false
   )
   &SpawnBurst_Instantaneous(
    Spawn Count=&Multiply_Int(
     A=&RandomRangeInt(Minimum=2, Maximum=5)
     B=&MakeIntFromFloat(Float=$System.InterpolateOverTime002.InterpolateOverTime_Float.Moving Average)
    )
    Use Spawn Probability=false
    UseLoopCountLimit=false
   )
   &SpawnRate(
    SpawnRate=&Multiply_FloatByInt(
     Float=2
     Integer=&MakeIntFromFloat(Float=$System.InterpolateOverTime002.InterpolateOverTime_Float.Moving Average)
    )
    Use Spawn Probability=false
   ) {enabled=false}
  ]
  particleSpawnScript [
   &InitializeParticle(
    Color=<rgba 1 0 0 1>
    Lifetime Max=0.4
    Lifetime Min=0.2
    Lifetime Mode=::NewEnumerator1
    Mesh Scale Max=<1.2 1.2 1.2>
    Mesh Scale Min=<0.8 0.8 0.8>
    Mesh Scale Mode=::NewEnumerator2
    UsePositionOffset=false
    Write Material Random=true
   )
   &InitialMeshOrientation(
    Mesh Axis To Orient=<0 0 1>
    Mesh Orientation Mode=::NewEnumerator2
    Rotation=&RandomRangeVector(Minimum=<-0.07 -0.07 -1>, Maximum=<0.07 0.07 1>)
    Use Rotation Vector=true
   )
   &ShapeLocation(
    Handle Radius=10
    Large Radius=35
    MaskSpawnGroup=false
    Override Seed=false
    Rotation Coordinate Space=::Simulation
    Shape Primitive=::NewEnumerator3
    UseOwnerScale=true
   )
   /SetVariables(
    Particles.LightningColor:LinearColor=<rgba 1 0 0 1>
    Particles.LightningScale:Vector3f=&VectorFromFloat(Value=&RandomRangeFloat(Minimum=0.4, Maximum=0.5))
   ) {isSetParameters=true, module=SetVariables_#7}
   /SetVariables(Particles.VisibilityTag=&RandomRangeInt(Minimum=0, Maximum=2)) {isSetParameters=true, module=SetVariables_#8}
  ]
  particleUpdateScript [
   &ParticleState()
   &SkeletalMeshLocation(
    Bone / Socket Index=&RandomRangeInt(Minimum=1, Maximum=4)
    Bone Sampling Mode=::NewEnumerator4
    MaskSpawnGroup=false
    Sampled Position Offset=<-15 -10 0>
    Skeletal Mesh=$User.SkeletalMesh
   )
   &DynamicMaterialParameters(
    Index 0 Param 1=&RandomRangeFloat(Minimum=2.5, Maximum=3)
    Index 0 Param 2=0.25
    Index 0 Param 3=$Particles.NormalizedAge
    Index 0 Param 4=&RandomRangeFloat(Minimum=0.85, Maximum=1)
    Index 1 Param 1=&RandomRangeFloat(Minimum=4, Maximum=5)
    Index 1 Param 2=0.5
    Index 1 Param 3=$Particles.NormalizedAge
    Index 1 Param 4=&RandomRangeFloat(Minimum=2, Maximum=2.4)
    Module.Param0WriteEnabled=true
    Module.Param0WriteEnabled1=true
    Module.Param1WriteEnabled=true
    Module.Param1WriteEnabled1=true
    Module.Param2WriteEnabled=true
    Module.Param2WriteEnabled1=true
    Module.Param3WriteEnabled=true
    Module.Param3WriteEnabled1=true
    Write Parameter Index 1=true
   )
   &UpdateMeshOrientation(Rotation Rate=0.25, Rotation Vector=<0 0 1>)
   &FN_Light_Attributes(
    Color=<rgba 10 0.1562 0.1562 1>
    Create Color Attribute=true
    Create Exponent Attribute=true
    Create Position Attribute=true
    Create Radius Attribute=true
    Enable Particle Light Enable=true
    Exponent=2
    Light Enabled:NiagaraBool=&SetBoolByIntComparison(
     A=&ModuloInt(A=&ReturnParticleIDAttributes(), B=3)
     B=1
     Comparison Type=::NewEnumerator4
    )
    Position=&AddVectorToPosition(
     Position=&AddVectorToPosition(
      Position=$Particles.SkeletalMeshLocation.SampledPosition
      Vector=&RotateVectorByQuaternion(Vector To Rotate=<-15 -10 0>, Quaternion=$Particles.SkeletalMeshLocation.SampledRotation)
     )
     Vector=&RotateVectorByQuaternion(
      Vector To Rotate=$Particles.ShapeLocation.ShapeVector
      Quaternion=&MultiplyQuaternion(
       Quaternion A=$Particles.SkeletalMeshLocation.SampledRotation
       Quaternion B=&FindQuatBetweenTwoVectors(From Vector=<1 0 0>, To Vector=<0 0 1>)
      )
     )
    )
    Radius=&FloatFromCurve(
     FloatCurve=DataInterface() {propertyValues {Curve {keys [
         <at 0 1 RCIM_Cubic>
         {arriveTangent=1.794, interpMode=RCIM_Cubic, leaveTangent=1.794, time=0.0746, value=0.7914}
         {arriveTangent=-0.0227, interpMode=RCIM_Cubic, leaveTangent=-0.0227, tangentMode=RCTM_User, time=0.121, value=1.2171}
         {arriveTangent=-5.5976, interpMode=RCIM_Cubic, leaveTangent=-5.5976, time=0.1377, value=0.3589}
         {arriveTangent=5.9152, interpMode=RCIM_Cubic, leaveTangent=5.9152, time=0.1738, value=0.9213}
         {arriveTangent=3.6344, interpMode=RCIM_Cubic, leaveTangent=3.6344, time=0.2093, value=0.7825}
         {arriveTangent=-4.8832, interpMode=RCIM_Cubic, leaveTangent=-4.8832, time=0.2396, value=1.1602}
         {arriveTangent=-6.1936, interpMode=RCIM_Cubic, leaveTangent=-6.1936, time=0.254, value=0.5643}
         {arriveTangent=-3.4254, interpMode=RCIM_Cubic, leaveTangent=-3.4254, time=0.2791, value=0.9154}
         {arriveTangent=1.2156, interpMode=RCIM_Cubic, leaveTangent=1.2156, time=0.3019, value=0.4002}
         {arriveTangent=5.6358, interpMode=RCIM_Cubic, leaveTangent=5.6358, time=0.3406, value=0.9901}
         {arriveTangent=6.3907, interpMode=RCIM_Cubic, leaveTangent=6.3907, time=0.3796, value=0.8384}
         {arriveTangent=-3.0158, interpMode=RCIM_Cubic, leaveTangent=-3.0158, time=0.4112, value=1.4411}
         {arriveTangent=-10.4214, interpMode=RCIM_Cubic, leaveTangent=-10.4214, time=0.4339, value=0.6747}
         {arriveTangent=-3.4821, interpMode=RCIM_Cubic, leaveTangent=-3.4821, time=0.4676, value=0.8529}
         {arriveTangent=-1.4741, interpMode=RCIM_Cubic, leaveTangent=-1.4741, time=0.4906, value=0.4775}
         {arriveTangent=-2.6963, interpMode=RCIM_Cubic, leaveTangent=-2.6963, time=0.5283, value=0.7634}
         {arriveTangent=2.8896, interpMode=RCIM_Cubic, leaveTangent=2.8896, time=0.5517, value=0.3126}
         {arriveTangent=6.2844, interpMode=RCIM_Cubic, leaveTangent=6.2844, time=0.5824, value=0.9196}
         {arriveTangent=1.4786, interpMode=RCIM_Cubic, leaveTangent=1.4786, time=0.5976, value=0.6006}
         {arriveTangent=4.8152, interpMode=RCIM_Cubic, leaveTangent=4.8152, time=0.6363, value=0.9993}
         {arriveTangent=2.8106, interpMode=RCIM_Cubic, leaveTangent=2.8106, time=0.6464, value=0.8356}
         {arriveTangent=3.2181, interpMode=RCIM_Cubic, leaveTangent=3.2181, time=0.6964, value=1.1681}
         {arriveTangent=1.8672, interpMode=RCIM_Cubic, leaveTangent=1.8672, time=0.7177, value=1.065}
         {arriveTangent=-12.2147, interpMode=RCIM_Cubic, leaveTangent=-12.2147, time=0.7385, value=1.2469}
         {arriveTangent=-0.0928, interpMode=RCIM_Cubic, leaveTangent=-0.0928, tangentMode=RCTM_User, time=0.7564, value=0.5916}
         {arriveTangent=3.085, interpMode=RCIM_Cubic, leaveTangent=3.085, time=0.7931, value=1.02}
         {arriveTangent=2.1222, interpMode=RCIM_Cubic, leaveTangent=2.1222, time=0.8092, value=0.7545}
         {arriveTangent=-0.424, interpMode=RCIM_Cubic, leaveTangent=-0.424, time=0.8924, value=1.2306}
         {arriveTangent=-8.511, interpMode=RCIM_Cubic, leaveTangent=-8.511, time=0.911, value=0.7114}
         {arriveTangent=3.2422, interpMode=RCIM_Cubic, leaveTangent=3.2422, time=0.9278, value=0.9291}
         {interpMode=RCIM_Cubic, leaveTangent=0.0463, time=1, value=1}
        ]}}}
     Scale Curve=50
    )
    Scale Color=&FloatFromCurve(
     FloatCurve=DataInterface() {propertyValues {Curve {keys [
         <at 0 1 RCIM_Cubic>
         {arriveTangent=1.794, interpMode=RCIM_Cubic, leaveTangent=1.794, time=0.0746, value=0.7914}
         {arriveTangent=-0.0227, interpMode=RCIM_Cubic, leaveTangent=-0.0227, tangentMode=RCTM_User, time=0.121, value=1.2171}
         {arriveTangent=-5.5976, interpMode=RCIM_Cubic, leaveTangent=-5.5976, time=0.1377, value=0.3589}
         {arriveTangent=5.9152, interpMode=RCIM_Cubic, leaveTangent=5.9152, time=0.1738, value=0.9213}
         {arriveTangent=3.6344, interpMode=RCIM_Cubic, leaveTangent=3.6344, time=0.2093, value=0.7825}
         {arriveTangent=-4.8832, interpMode=RCIM_Cubic, leaveTangent=-4.8832, time=0.2396, value=1.1602}
         {arriveTangent=-6.1936, interpMode=RCIM_Cubic, leaveTangent=-6.1936, time=0.254, value=0.5643}
         {arriveTangent=-3.4254, interpMode=RCIM_Cubic, leaveTangent=-3.4254, time=0.2791, value=0.9154}
         {arriveTangent=1.2156, interpMode=RCIM_Cubic, leaveTangent=1.2156, time=0.3019, value=0.4002}
         {arriveTangent=5.6358, interpMode=RCIM_Cubic, leaveTangent=5.6358, time=0.3406, value=0.9901}
         {arriveTangent=6.3907, interpMode=RCIM_Cubic, leaveTangent=6.3907, time=0.3796, value=0.8384}
         {arriveTangent=-3.0158, interpMode=RCIM_Cubic, leaveTangent=-3.0158, time=0.4112, value=1.4411}
         {arriveTangent=-10.4214, interpMode=RCIM_Cubic, leaveTangent=-10.4214, time=0.4339, value=0.6747}
         {arriveTangent=-3.4821, interpMode=RCIM_Cubic, leaveTangent=-3.4821, time=0.4676, value=0.8529}
         {arriveTangent=-1.4741, interpMode=RCIM_Cubic, leaveTangent=-1.4741, time=0.4906, value=0.4775}
         {arriveTangent=-2.6963, interpMode=RCIM_Cubic, leaveTangent=-2.6963, time=0.5283, value=0.7634}
         {arriveTangent=2.8896, interpMode=RCIM_Cubic, leaveTangent=2.8896, time=0.5517, value=0.3126}
         {arriveTangent=6.2844, interpMode=RCIM_Cubic, leaveTangent=6.2844, time=0.5824, value=0.9196}
         {arriveTangent=1.4786, interpMode=RCIM_Cubic, leaveTangent=1.4786, time=0.5976, value=0.6006}
         {arriveTangent=4.8152, interpMode=RCIM_Cubic, leaveTangent=4.8152, time=0.6363, value=0.9993}
         {arriveTangent=2.8106, interpMode=RCIM_Cubic, leaveTangent=2.8106, time=0.6464, value=0.8356}
         {arriveTangent=3.2181, interpMode=RCIM_Cubic, leaveTangent=3.2181, time=0.6964, value=1.1681}
         {arriveTangent=1.8672, interpMode=RCIM_Cubic, leaveTangent=1.8672, time=0.7177, value=1.065}
         {arriveTangent=-12.2147, interpMode=RCIM_Cubic, leaveTangent=-12.2147, time=0.7385, value=1.2469}
         {arriveTangent=-0.0928, interpMode=RCIM_Cubic, leaveTangent=-0.0928, tangentMode=RCTM_User, time=0.7564, value=0.5916}
         {arriveTangent=3.085, interpMode=RCIM_Cubic, leaveTangent=3.085, time=0.7931, value=1.02}
         {arriveTangent=2.1222, interpMode=RCIM_Cubic, leaveTangent=2.1222, time=0.8092, value=0.7545}
         {arriveTangent=-0.424, interpMode=RCIM_Cubic, leaveTangent=-0.424, time=0.8924, value=1.2306}
         {arriveTangent=-8.511, interpMode=RCIM_Cubic, leaveTangent=-8.511, time=0.911, value=0.7114}
         {arriveTangent=3.2422, interpMode=RCIM_Cubic, leaveTangent=3.2422, time=0.9278, value=0.9291}
         {interpMode=RCIM_Cubic, leaveTangent=0.0463, time=1, value=1}
        ]}}}
     Scale Curve=3
    )
   )
  ]
  rendererClasses [&NiagaraMeshRendererProperties]
  renderers [
   {rendererClass=&NiagaraMeshRendererProperties, rendererIndex=0}
   {rendererClass=&NiagaraMeshRendererProperties, rendererIndex=1}
   {rendererClass=&NiagaraMeshRendererProperties, rendererIndex=2}
  ]
 }
 Smoke {
  emitterUpdateScript [
   &EmitterState()
   &SpawnParticlesFromOtherEmitter(
    Attribute Reader=DataInterface() {propertyValues {EmitterBinding {bindingMode=Other, emitterName=Soul}}}
    Spawn Rate=&Multiply_Float(A=40, B=$System.InterpolateOverTime.InterpolateOverTime_Float.Moving Average)
    Use Maximum Age=false
    Use Minimum Age=false
   )
  ]
  particleSpawnScript [
   &InitializeParticle(
    Lifetime Max=1.75
    Lifetime Min=1.5
    Lifetime Mode=::NewEnumerator1
    Sprite Rotation Mode=::NewEnumerator2
    Sprite Size Mode=::NewEnumerator1
    Uniform Sprite Size Max=75
    Uniform Sprite Size Min=50
    UsePositionOffset=false
    Write Material Random=true
   )
   &SampleParticlesFromOtherEmitter(Mesh Orientation Sampling:ENiagara_DirectReadSamplingMode=::NewEnumerator0)
   &CameraOffset(Camera Offset Amount=&RandomRangeFloat(Minimum=50, Maximum=100))
  ]
  particleUpdateScript [
   &ParticleState()
   &ShapeLocation(
    Additional Quaternion Rotation=false
    Apply To Particle Position=::NewEnumerator1
    MaskSpawnGroup=false
    Non Uniform Scale=<0.8 1 2>
    Offset=<0 0 12>
    Offset Mode=::NewEnumerator0
    Override Seed=false
    Rotation Mode=::NewEnumerator3
    Rotation Quaternion=&MultiplyQuaternion(
     Quaternion A=&self:Get_MeshOrientation_Attribute_By_Index(
      Particle Attribute Reader:NiagaraDataInterfaceParticleRead=DataInterface() {propertyValues {EmitterBinding {bindingMode=Other, emitterName=Soul}}}
      Particle Index:NiagaraInt32=&MakeIntFromFloat(Float=$Particles.SampleParticlesFromOtherEmitter.SampledIndex)
     )
     Quaternion B=&FindQuatBetweenTwoVectors(From Vector=<1 0 0>, To Vector=<1 0 1>)
    )
    Sphere Radius=8
    Transform Order:ENiagara_TransformOrder=::NewEnumerator1
    UseOwnerScale=true
   )
   &InheritSourceMovement(
    Applied Position Scale=&VectorFromFloat(Value=&FloatFromCurve(FloatCurve=DataInterface() {propertyValues {Curve {keys [<at 0 1>,<at 0.5 0.9>]}}}))
    ApplyForce=false
    ApplyPosition=true
    ApplyPositionFalloff=false
   )
   &SolveForcesAndVelocity(Clamp Velocity=false, Limit Acceleration=false)
   &ScaleSpriteSize(Scale Factor=&Vector2DFromFloat(Value=&FloatFromCurve(FloatCurve=DataInterface() {propertyValues {Curve {keys [
          {arriveTangent=1.8751, interpMode=RCIM_Cubic, time=0, value=0.4}
          {arriveTangent=2.3763, interpMode=RCIM_Cubic, leaveTangent=2.3763, tangentMode=RCTM_User, time=0.5, value=1}
          {arriveTangent=0.0421, interpMode=RCIM_Cubic, leaveTangent=0.0421, tangentMode=RCTM_User, time=1, value=2}
         ]}}})))
   &ScaleColor(
    Linear Color Curve=DataInterface() {propertyValues {
      AlphaCurve {keys [<at 0 1>,<at 0.36 1>,<at 0.553 0.25>,<at 0.8768 0>]}
      BlueCurve {keys [<at 0.0686 0.1>,<at 0.1959 0.04>,<at 0.4335 0.0674>]}
      GreenCurve {keys [<at 0.0686 2>,<at 0.1959 0.8>,<at 0.4335 0.1033>]}
      RedCurve {keys [<at 0.0686 10>,<at 0.1959 4>,<at 0.4335 0.1771>]}
     }}
    Scale Mode=::NewEnumerator2
   )
   &ScaleColor(Scale Alpha=$System.InterpolateOverTime.InterpolateOverTime_Float.Moving Average, ScaleA=true, ScaleRGB=false) {module=ScaleColor001}
   &SpriteRotationRate(Rotation Rate=&UniformRangedFloat(Minimum=-15, Maximum=15))
   &DynamicMaterialParameters(
    Index 0 Param 1=&FloatFromCurve(FloatCurve=DataInterface() {propertyValues {Curve {keys [
         {arriveTangent=-7.0802, interpMode=RCIM_Cubic, leaveTangent=-7.0802, tangentMode=RCTM_User, time=0, value=2}
         <at 0.5 0.05 RCIM_Cubic>
         <at 1 2 RCIM_Cubic>
        ]}}})
    Index 0 Param 2=&FloatFromCurve(FloatCurve=DataInterface() {propertyValues {Curve {keys [
         {interpMode=RCIM_Cubic, tangentMode=RCTM_User, time=0, value=0}
         {interpMode=RCIM_Cubic, tangentMode=RCTM_User, time=0.5, value=1}
         {tangentMode=RCTM_User, time=1, value=0}
        ]}}})
    Index 0 Param 3=&RandomRangeFloat(Minimum=0.6, Maximum=0.8)
    Index 0 Param 4=150
    Index 1 Param 1=150
    Index 1 Param 2=200
    Module.Param0WriteEnabled=true
    Module.Param0WriteEnabled1=true
    Module.Param1WriteEnabled=true
    Module.Param1WriteEnabled1=true
    Module.Param2WriteEnabled=true
    Module.Param2WriteEnabled1=false
    Module.Param3WriteEnabled=false
    Module.Param3WriteEnabled1=false
    Write Parameter Index 1=true
   )
   &SubUVAnimation()
   /SetVariables(Particles.Position=&Lerp_Position(
     A:NiagaraPosition=&Add_Vector(
      A=&self:Get_Position_Attribute_By_Index(
       Particle Attribute Reader:NiagaraDataInterfaceParticleRead=DataInterface() {propertyValues {EmitterBinding {bindingMode=Other, emitterName=Soul}}}
       Particle Index:NiagaraInt32=&MakeIntFromFloat(Float=$Particles.SampleParticlesFromOtherEmitter.SampledIndex)
      )
      B=$Particles.ShapeLocation.ShapeVector
     )
     B:NiagaraPosition=&AddVectorToPosition(
      Position=$Particles.Position
      Vector=&RotateVectorByQuaternion(
       Vector To Rotate=<0 0 5>
       Quaternion=&MultiplyQuaternion(
        Quaternion A=&self:Get_MeshOrientation_Attribute_By_Index(
         Particle Attribute Reader:NiagaraDataInterfaceParticleRead=DataInterface() {propertyValues {EmitterBinding {bindingMode=Other, emitterName=Soul}}}
         Particle Index:NiagaraInt32=&MakeIntFromFloat(Float=$Particles.SampleParticlesFromOtherEmitter.SampledIndex)
        )
        Quaternion B=&FindQuatBetweenTwoVectors(From Vector=<1 0 0>, To Vector=<1 0 1>)
       )
      )
     )
     Alpha:NiagaraFloat=&FloatFromCurve(FloatCurve=DataInterface() {propertyValues {Curve {keys [
          {arriveTangent=1.9048, interpMode=RCIM_Cubic, leaveTangent=1.9048, tangentMode=RCTM_User, time=0.1, value=0}
          {arriveTangent=0.5863, interpMode=RCIM_Cubic, leaveTangent=0.5863, tangentMode=RCTM_User, time=1, value=1}
         ]}}})
    )) {isSetParameters=true, module=SetVariables_#6}
  ]
  rendererClasses [&NiagaraSpriteRendererProperties]
  renderers [{rendererClass=&NiagaraSpriteRendererProperties}]
 }
 Smoke_FE {
  emitterUpdateScript [
   &EmitterState()
   &SpawnParticlesFromOtherEmitter(
    Attribute Reader=DataInterface() {propertyValues {EmitterBinding {bindingMode=Other, emitterName=Soul_FE}}}
    Spawn Rate=&Multiply_Float(A=40, B=$System.InterpolateOverTime.InterpolateOverTime_Float.Moving Average)
    Use Maximum Age=false
    Use Minimum Age=false
   )
  ]
  particleSpawnScript [
   &InitializeParticle(
    Lifetime Max=1.75
    Lifetime Min=1.5
    Lifetime Mode=::NewEnumerator1
    Sprite Rotation Mode=::NewEnumerator2
    Sprite Size Mode=::NewEnumerator1
    Uniform Sprite Size Max=75
    Uniform Sprite Size Min=50
    UsePositionOffset=false
    Write Material Random=true
   )
   &SampleParticlesFromOtherEmitter(Mesh Orientation Sampling:ENiagara_DirectReadSamplingMode=::NewEnumerator0)
   &CameraOffset(Camera Offset Amount=&RandomRangeFloat(Minimum=50, Maximum=100))
  ]
  particleUpdateScript [
   &ParticleState()
   &ShapeLocation(
    Additional Quaternion Rotation=false
    Apply To Particle Position=::NewEnumerator1
    MaskSpawnGroup=false
    Non Uniform Scale=<0.8 1 2>
    Offset=<0 0 12>
    Offset Mode=::NewEnumerator0
    Override Seed=false
    Rotation Mode=::NewEnumerator3
    Rotation Quaternion=&MultiplyQuaternion(
     Quaternion A=&self:Get_MeshOrientation_Attribute_By_Index(
      Particle Attribute Reader:NiagaraDataInterfaceParticleRead=DataInterface() {propertyValues {EmitterBinding {bindingMode=Other, emitterName=Soul_FE}}}
      Particle Index:NiagaraInt32=&MakeIntFromFloat(Float=$Particles.SampleParticlesFromOtherEmitter.SampledIndex)
     )
     Quaternion B=&FindQuatBetweenTwoVectors(From Vector=<1 0 0>, To Vector=<1 0 1>)
    )
    Sphere Radius=8
    Transform Order:ENiagara_TransformOrder=::NewEnumerator1
    UseOwnerScale=true
   )
   &InheritSourceMovement(
    Applied Position Scale=&VectorFromFloat(Value=&FloatFromCurve(FloatCurve=DataInterface() {propertyValues {Curve {keys [<at 0 1>,<at 0.5 0.9>]}}}))
    ApplyForce=false
    ApplyPosition=true
    ApplyPositionFalloff=false
   )
   &SolveForcesAndVelocity(Clamp Velocity=false, Limit Acceleration=false)
   &ScaleSpriteSize(Scale Factor=&Vector2DFromFloat(Value=&FloatFromCurve(FloatCurve=DataInterface() {propertyValues {Curve {keys [
          {arriveTangent=1.8751, interpMode=RCIM_Cubic, time=0, value=0.4}
          {arriveTangent=2.3763, interpMode=RCIM_Cubic, leaveTangent=2.3763, tangentMode=RCTM_User, time=0.5, value=1}
          {arriveTangent=0.0421, interpMode=RCIM_Cubic, leaveTangent=0.0421, tangentMode=RCTM_User, time=1, value=2}
         ]}}})))
   &ScaleColor(
    Linear Color Curve=DataInterface() {propertyValues {
      AlphaCurve {keys [<at 0 1>,<at 0.36 1>,<at 0.553 0.25>,<at 0.8768 0>]}
      BlueCurve {keys [<at 0.0686 0.1>,<at 0.1959 0.04>,<at 0.4335 0.0674>]}
      GreenCurve {keys [<at 0.0686 2>,<at 0.1959 0.8>,<at 0.4335 0.1033>]}
      RedCurve {keys [<at 0.0686 10>,<at 0.1959 4>,<at 0.4335 0.1771>]}
     }}
    Scale Mode=::NewEnumerator2
   )
   &ScaleColor(Scale Alpha=$System.InterpolateOverTime.InterpolateOverTime_Float.Moving Average, ScaleA=true, ScaleRGB=false) {module=ScaleColor001}
   &SpriteRotationRate(Rotation Rate=&UniformRangedFloat(Minimum=-15, Maximum=15))
   &DynamicMaterialParameters(
    Index 0 Param 1=&FloatFromCurve(FloatCurve=DataInterface() {propertyValues {Curve {keys [
         {arriveTangent=-7.0802, interpMode=RCIM_Cubic, leaveTangent=-7.0802, tangentMode=RCTM_User, time=0, value=2}
         <at 0.5 0.05 RCIM_Cubic>
         <at 1 2 RCIM_Cubic>
        ]}}})
    Index 0 Param 2=&FloatFromCurve(FloatCurve=DataInterface() {propertyValues {Curve {keys [
         {interpMode=RCIM_Cubic, tangentMode=RCTM_User, time=0, value=0}
         {interpMode=RCIM_Cubic, tangentMode=RCTM_User, time=0.5, value=1}
         {tangentMode=RCTM_User, time=1, value=0}
        ]}}})
    Index 0 Param 3=&RandomRangeFloat(Minimum=0.6, Maximum=0.8)
    Index 0 Param 4=150
    Index 1 Param 1=50
    Index 1 Param 2=300
    Module.Param0WriteEnabled=true
    Module.Param0WriteEnabled1=true
    Module.Param1WriteEnabled=true
    Module.Param1WriteEnabled1=true
    Module.Param2WriteEnabled=true
    Module.Param2WriteEnabled1=false
    Module.Param3WriteEnabled=false
    Module.Param3WriteEnabled1=false
    Write Parameter Index 1=true
   )
   &SubUVAnimation()
   /SetVariables(Particles.Position=&Lerp_Position(
     A:NiagaraPosition=&Add_Vector(
      A=&self:Get_Position_Attribute_By_Index(
       Particle Attribute Reader:NiagaraDataInterfaceParticleRead=DataInterface() {propertyValues {EmitterBinding {bindingMode=Other, emitterName=Soul_FE}}}
       Particle Index:NiagaraInt32=&MakeIntFromFloat(Float=$Particles.SampleParticlesFromOtherEmitter.SampledIndex)
      )
      B=$Particles.ShapeLocation.ShapeVector
     )
     B:NiagaraPosition=&AddVectorToPosition(
      Position=$Particles.Position
      Vector=&RotateVectorByQuaternion(
       Vector To Rotate=<0 0 200>
       Quaternion=&MultiplyQuaternion(
        Quaternion A=&self:Get_MeshOrientation_Attribute_By_Index(
         Particle Attribute Reader:NiagaraDataInterfaceParticleRead=DataInterface() {propertyValues {EmitterBinding {bindingMode=Other, emitterName=Soul_FE}}}
         Particle Index:NiagaraInt32=&MakeIntFromFloat(Float=$Particles.SampleParticlesFromOtherEmitter.SampledIndex)
        )
        Quaternion B=&FindQuatBetweenTwoVectors(From Vector=<1 0 0>, To Vector=<0.5 0 1>)
       )
      )
     )
     Alpha:NiagaraFloat=&FloatFromCurve(FloatCurve=DataInterface() {propertyValues {Curve {keys [{interpMode=RCIM_Cubic, tangentMode=RCTM_User, time=0.05, value=0}, {interpMode=RCIM_Cubic, tangentMode=RCTM_User, time=1, value=1}]}}})
    )) {isSetParameters=true, module=SetVariables_#6}
  ]
  rendererClasses [&NiagaraSpriteRendererProperties]
  renderers [{rendererClass=&NiagaraSpriteRendererProperties}]
 }
 Soul {
  emitterUpdateScript [
   &EmitterState()
   &SpawnBurst_Instantaneous(
    Spawn Count=&MakeIntFromFloat(Float=&FN_FrontEndFloatModifier(Gameplay=3, Front End=0))
    Use Spawn Probability=false
    UseLoopCountLimit=true
   )
   &InterpolateOverTime(Rate Of Change=50, Target Vector Value:Vector3f=$Engine.Owner.Velocity, Variable Type=::NewEnumerator2)
  ]
  particleSpawnScript [
   &FN_SetSpawnGroup(Number of Spawn Groups=3)
   &InitializeParticle(
    Lifetime=2
    Mesh Scale Mode=::NewEnumerator3
    Mesh Uniform Scale=$System.Ship Scale
    Sprite Size Mode=::NewEnumerator3
    Uniform Sprite Size=40
    UsePositionOffset=true
    Write Material Random=true
   )
  ]
  particleUpdateScript [
   &ParticleState(Kill Particles When Lifetime Has Elapsed=false, Let Infinitely Lived Particles Die When Emitter Deactivates=true)
   /SetVariables(Particles.Ambient:Vector3f=&Multiply_VectorByFloat(
     Vector=&MakeVector(
      X=&Waveform(
       "Waveform [1]"=::NewEnumerator12
       "[1] Input Position"=&RandomRangeFloat(Minimum=0, Maximum=1000)
       "[1] Frequency"=1
       "[1] Amplitude Min/Max"=<-50 50>
      )
      Y=&Waveform(
       "Waveform [1]"=::NewEnumerator12
       "[1] Input Position"=&RandomRangeFloat(Minimum=1200, Maximum=1800)
       "[1] Frequency"=1
       "[1] Amplitude Min/Max"=<-50 50>
      )
      Z=&Waveform(
       "Waveform [1]"=::NewEnumerator12
       "[1] Input Position"=&RandomRangeFloat(Minimum=2200, Maximum=2800)
       "[1] Frequency"=1
       "[1] Amplitude Min/Max"=<-50 50>
      )
     )
     Float=0.35
    )) {isSetParameters=true, module=SetVariables_#9}
   &InterpolateOverTime(
    Rate Of Change=3
    Target Float Value:NiagaraFloat=&MakeCustomFloatFromBool(Bool=$System.Timeline001.IsRewinding, True Float=0, False Float=1)
   ) {module=InterpolateOverTime004}
   /SetVariables(Particles.CatchupVector:Vector3f=&Multiply_VectorByFloat(
     Vector=&Multiply_VectorByFloat(
      Vector=&NormalizeVector(Vector To Normalize=$System.SkydiveVelocity, Safe Normalize=true, Safe Fallback Threshold=0, Safe Fallback Vector=<0 0 0.001>)
      Float=-75
     )
     Float=&Lerp_Float(
      A=&FloatFromCurve(
       FloatCurve=DataInterface() {propertyValues {Curve {keys [
           {interpMode=RCIM_Cubic, tangentMode=RCTM_User, time=0, value=0}
           {interpMode=RCIM_Cubic, tangentMode=RCTM_User, time=0.35, value=-0.35}
           {tangentMode=RCTM_User, time=1, value=0}
          ]}}}
       CurveIndex=$System.Timeline001.NormalizedTime
      )
      B=&FloatFromCurve(
       FloatCurve=DataInterface() {propertyValues {Curve {keys [
           {arriveTangent=3.3827, arriveTangentWeight=1.1106, interpMode=RCIM_Cubic, leaveTangentWeight=1.1106, time=0, value=0}
           {arriveTangentWeight=0.5241, interpMode=RCIM_Cubic, leaveTangentWeight=0.1667, time=0.5, value=0.65}
           <at 1 0 RCIM_Cubic>
          ]}}}
       CurveIndex=$System.Timeline001.NormalizedTime
      )
      Alpha=$Particles.InterpolateOverTime004.InterpolateOverTime_Float.Moving Average
     )
    )) {isSetParameters=true, module=SetVariables_#10}
   &ShapeLocation(
    Apply To Particle Position=::NewEnumerator1
    MaskSpawnGroup=false
    Override Seed=false
    Ring / Disc Distribution Mode:ENiagaraSphereDistributionMode=::NewEnumerator1
    Ring Radius=&Lerp_Float(A=100, B=125, Alpha=$System.InterpolateOverTime.InterpolateOverTime_Float.Moving Average)
    Rotation Angle=0.75
    Rotation Axis=<0 1 0>
    Rotation Coordinate Space=::World
    Rotation Mode=::NewEnumerator1
    Shape Primitive=::NewEnumerator4
    U Position=&ReturnNormalizedExecIndex(Normalized Index Includes One=false)
    UseOwnerScale=true
   )
   &SkeletalMeshLocation(
    Bone Sampling Mode=::NewEnumerator4
    MaskSpawnGroup=false
    Orientation Sampling=::NewEnumerator1
    Sampled Position Offset=&Add_Vector(
     A=$Particles.CatchupVector
     B=/Niagara/DynamicInputs/Vectors/RotateVector(
      Vector To Rotate:Vector3f=&Add_Vector(
       A=$Particles.Ambient
       B=&Add_Vector(
        A=&Add_Vector(
         A=$Particles.ShapeLocation.ShapeVector
         B=&MaskVectorBySpawnGroup(
          Masked Vector=&Lerp_Vector(A=<0 110 -50>, B=<0 110 -150>, Alpha=$System.InterpolateOverTime.InterpolateOverTime_Float.Moving Average)
          Passthrough Vector=<0 0 0>
         )
        )
        B=$System.OverallTargetOffset
       )
      )
      Yaw:NiagaraFloat=&Waveform(
       "Waveform [1]"=::NewEnumerator12
       "[1] Input Position"=&RandomRangeFloat(Minimum=80000, Maximum=40000)
       "[1] Frequency"=1.25
       "[1] Amplitude Min/Max"=<7 -7>
      )
      Pitch:NiagaraFloat=0
      Roll:NiagaraFloat=0
      Rotation Rate:NiagaraFloat=2
     )
    )
    Skeletal Mesh=$User.SkeletalMesh
    Use Sampled Mesh Attribute as Ribbon ID=false
   )
   &ScaleMeshSize(Scale Factor=&VectorFromFloat(Value=&FloatFromCurve(
      FloatCurve=DataInterface() {propertyValues {Curve {keys [{arriveTangent=1.0186, interpMode=RCIM_Cubic, leaveTangent=1.0186, tangentMode=RCTM_User, time=0, value=0}, <at 1 1 RCIM_Cubic>]}}}
      CurveIndex=$Emitter.Age
     )))
   /SetVariables(Particles.Wobble:Quat4f=&MultiplyQuaternion(
     Quaternion A=&FindQuatBetweenTwoVectors(
      From Vector=&MakeVector(
       X=&Sine(
        Normalized Angle=&Add_Float(A=$System.FN_Integrator.Value, B=&RandomRangeFloat(Minimum=0, Maximum=1))
        Period=&MakeCustomFloatFromBool(Bool=&RandomBool(), True Float=1, False Float=-1)
        Scale=1
        Bias=0
       )
       Y=&Cosine(
        Normalized Angle=&Add_Float(A=$System.FN_Integrator.Value, B=&RandomRangeFloat(Minimum=0, Maximum=1))
        Period=&MakeCustomFloatFromBool(Bool=&RandomBool(), True Float=1, False Float=-1)
        Scale=1
        Bias=0
       )
       Z=&RandomRangeFloat(Minimum=6, Maximum=8)
      )
      To Vector=<0 0 1>
     )
     Quaternion B=$Particles.Initial.MeshOrientation
    )) {isSetParameters=true, module=SetVariables_#11}
   &UpdateMeshOrientation(
    Bank Rate:NiagaraFloat=&Lerp_Float(
     A=&Lerp_Float(
      A=1
      B=30
      Alpha=&Waveform(
       "Waveform [1]"=::NewEnumerator12
       "[1] Input Position"=&RandomRangeFloat(Minimum=0, Maximum=1000)
       "[1] Phase Scale"=50000
       "[1] Frequency"=0.5
       "[1] Amplitude Min/Max"=<0 1>
      )
     )
     B=900
     Alpha=$System.Dive
    )
    Banking Angle Min / Max:Vector2f=&Lerp_Vector2(A=<-10 10>, B=<-4 4>, Alpha=$System.Dive)
    "Foward Vector (Mesh Space)":Vector3f=&Lerp_Vector(A=<0.5 0 1>, B=<1 0 0>, Alpha=$System.InterpolateOverTime.InterpolateOverTime_Float.Moving Average)
    Orientation Method=::NewEnumerator3
    Particle Velocity:Vector3f=$Emitter.InterpolateOverTime.InterpolateOverTime_V3_.Moving Average
    Rotation Decay Rate In Degrees:NiagaraFloat=&Lerp_Float(A=150, B=500, Alpha=$System.Dive)
    Up Vector:Vector3f=&Lerp_Vector(
     A=&ReturnMeshOrientationAxis(Mesh Orientation Axis=::YAxis, Mesh Orientation Quat=$Particles.SkeletalMeshLocation.SampledRotation)
     B=&Multiply_VectorByFloat(
      Vector=&ReturnMeshOrientationAxis(Mesh Orientation Axis=::ZAxis, Mesh Orientation Quat=$Particles.SkeletalMeshLocation.SampledRotation)
      Float=1
     )
     Alpha=&FloatFromCurve(FloatCurve=DataInterface() {propertyValues {Curve {keys [<at 0 0.5>,<at 0.25 0.75>]}}}, CurveIndex=$System.Dive)
    )
    Upright Roll Correction Speed:NiagaraFloat=&Lerp_Float(A=25, B=90000, Alpha=$System.Dive)
    Velocity Orientation Convergence Rate:NiagaraFloat=&Lerp_Float(
     A=&Lerp_Float(
      A=3
      B=12
      Alpha=&Waveform(
       "Waveform [1]"=::NewEnumerator12
       "[1] Input Position"=&RandomRangeFloat(Minimum=0, Maximum=100)
       "[1] Frequency"=5
       "[1] Amplitude Min/Max"=<0 1>
      )
     )
     B=&Lerp_Float(
      A=20
      B=40
      Alpha=&Waveform(
       "Waveform [1]"=::NewEnumerator12
       "[1] Input Position"=&RandomRangeFloat(Minimum=0, Maximum=100)
       "[1] Frequency"=5
       "[1] Amplitude Min/Max"=<0 1>
      )
     )
     Alpha=$System.Dive
    )
   ) {enabled=false}
   &/UpdateMeshOrientation_RedPepper_TEMP(
    Bank Rate:NiagaraFloat=&Lerp_Float(
     A=&Lerp_Float(
      A=1
      B=30
      Alpha=&Waveform(
       "Waveform [1]"=::NewEnumerator12
       "[1] Input Position"=&RandomRangeFloat(Minimum=0, Maximum=1000)
       "[1] Phase Scale"=50000
       "[1] Frequency"=0.5
       "[1] Amplitude Min/Max"=<0 1>
      )
     )
     B=900
     Alpha=$System.Dive
    )
    Banking Angle Min / Max:Vector2f=&Lerp_Vector2(A=<-10 10>, B=<-4 4>, Alpha=$System.Dive)
    "Foward Vector (Mesh Space)":Vector3f=&Lerp_Vector(A=<0.5 0 1>, B=<1 0 0>, Alpha=$System.InterpolateOverTime.InterpolateOverTime_Float.Moving Average)
    Orientation Method:ENiagara_UpdateMeshOrientationMode=::NewEnumerator3
    Particle Velocity:Vector3f=$Emitter.InterpolateOverTime.InterpolateOverTime_V3_.Moving Average
    Rotation Decay Rate In Degrees:NiagaraFloat=&Lerp_Float(A=150, B=500, Alpha=$System.Dive)
    Up Vector:Vector3f=&Lerp_Vector(
     A=&ReturnMeshOrientationAxis(Mesh Orientation Axis=::YAxis, Mesh Orientation Quat=$Particles.SkeletalMeshLocation.SampledRotation)
     B=&Multiply_VectorByFloat(
      Vector=&ReturnMeshOrientationAxis(Mesh Orientation Axis=::ZAxis, Mesh Orientation Quat=$Particles.SkeletalMeshLocation.SampledRotation)
      Float=1
     )
     Alpha=&FloatFromCurve(FloatCurve=DataInterface() {propertyValues {Curve {keys [<at 0 0.5>,<at 0.25 0.75>]}}}, CurveIndex=$System.Dive)
    )
    Upright Roll Correction Speed:NiagaraFloat=&Lerp_Float(A=25, B=90000, Alpha=$System.Dive)
    Velocity Orientation Convergence Rate:NiagaraFloat=&Lerp_Float(
     A=&Lerp_Float(
      A=3
      B=12
      Alpha=&Waveform(
       "Waveform [1]"=::NewEnumerator12
       "[1] Input Position"=&RandomRangeFloat(Minimum=0, Maximum=100)
       "[1] Frequency"=5
       "[1] Amplitude Min/Max"=<0 1>
      )
     )
     B=&Lerp_Float(
      A=20
      B=40
      Alpha=&Waveform(
       "Waveform [1]"=::NewEnumerator12
       "[1] Input Position"=&RandomRangeFloat(Minimum=0, Maximum=100)
       "[1] Frequency"=5
       "[1] Amplitude Min/Max"=<0 1>
      )
     )
     Alpha=$System.Dive
    )
   )
   &InterpolateOverTime(Rate Of Change=5, Target Quat Value:Quat4f=$Particles.MeshOrientation, Variable Type=::NewEnumerator5) {module=InterpolateOverTime002}
   /SetVariables(Particles.New Orient:Quat4f=$Particles.InterpolateOverTime002.InterpolateOverTime_Quaternion.Moving Average) {isSetParameters=true, module=SetVariables_#12}
   /SetVariables(Particles.MeshOrientation=&MultiplyQuaternion(Quaternion A=$Particles.Wobble, Quaternion B=$Particles.New Orient)) {isSetParameters=true, module=SetVariables_#13}
   &ScaleColor(
    Scale RGB=&MakeVectorFromLinearColorRGB(Color=&LerpLinearColors(StartColor=<rgba 0 1.3565 2 1>, EndColor=<rgba 4 0.479 0 1>, LerpFactor=$System.InterpolateOverTime.InterpolateOverTime_Float.Moving Average))
    ScaleA=true
    ScaleRGB=true
   )
   &DynamicMaterialParameters(
    Index 0 Param 1=1
    Index 0 Param 2=0.5
    Index 0 Param 3=&RandomRangeFloat(Minimum=0.1, Maximum=1)
    Index 0 Param 4=&RandomRangeFloat(Minimum=0.75, Maximum=1)
    Module.Param0WriteEnabled=true
    Module.Param1WriteEnabled=true
    Module.Param2WriteEnabled=true
    Module.Param3WriteEnabled=true
   )
   /SetVariables(
    Particles.Decal Color:LinearColor=&LerpLinearColors(
     StartColor=<rgba 0 1.3565 2 3>
     EndColor=<rgba 6 0.9039 0 3>
     LerpFactor=$System.InterpolateOverTime.InterpolateOverTime_Float.Moving Average
    )
    Particles.Decal Offset=15
    Particles.Decal Position:NiagaraPosition=&AddVectorToPosition(
     Position=$Particles.Position
     Vector=&RotateVectorByQuaternion(Vector To Rotate=<-10 0 15>, Quaternion=$Particles.MeshOrientation)
    )
   ) {isSetParameters=true, module=SetVariables_#14}
   &FN_Light_Attributes(
    Color=&LerpLinearColors(
     StartColor=<rgba 0 1.3565 2 1>
     EndColor=<rgba 6 0.9039 0 4>
     LerpFactor=$System.InterpolateOverTime.InterpolateOverTime_Float.Moving Average
    )
    Create Color Attribute=true
    Create Exponent Attribute=true
    Create Radius Attribute=true
    Exponent=2
    Radius=85
    Scale Color=&FloatFromCurve(
     FloatCurve=DataInterface() {propertyValues {Curve {keys [
         <at 0 1 RCIM_Cubic>
         {arriveTangent=1.794, interpMode=RCIM_Cubic, leaveTangent=1.794, time=0.0746, value=0.7914}
         {arriveTangent=-0.0227, interpMode=RCIM_Cubic, leaveTangent=-0.0227, tangentMode=RCTM_User, time=0.121, value=1.2171}
         {arriveTangent=-5.5976, interpMode=RCIM_Cubic, leaveTangent=-5.5976, time=0.1377, value=0.3589}
         {arriveTangent=5.9152, interpMode=RCIM_Cubic, leaveTangent=5.9152, time=0.1738, value=0.9213}
         {arriveTangent=3.6344, interpMode=RCIM_Cubic, leaveTangent=3.6344, time=0.2093, value=0.7825}
         {arriveTangent=-4.8832, interpMode=RCIM_Cubic, leaveTangent=-4.8832, time=0.2396, value=1.1602}
         {arriveTangent=-6.1936, interpMode=RCIM_Cubic, leaveTangent=-6.1936, time=0.254, value=0.5643}
         {arriveTangent=-3.4254, interpMode=RCIM_Cubic, leaveTangent=-3.4254, time=0.2791, value=0.9154}
         {arriveTangent=1.2156, interpMode=RCIM_Cubic, leaveTangent=1.2156, time=0.3019, value=0.4002}
         {arriveTangent=5.6358, interpMode=RCIM_Cubic, leaveTangent=5.6358, time=0.3406, value=0.9901}
         {arriveTangent=6.3907, interpMode=RCIM_Cubic, leaveTangent=6.3907, time=0.3796, value=0.8384}
         {arriveTangent=-3.0158, interpMode=RCIM_Cubic, leaveTangent=-3.0158, time=0.4112, value=1.4411}
         {arriveTangent=-10.4214, interpMode=RCIM_Cubic, leaveTangent=-10.4214, time=0.4339, value=0.6747}
         {arriveTangent=-3.4821, interpMode=RCIM_Cubic, leaveTangent=-3.4821, time=0.4676, value=0.8529}
         {arriveTangent=-1.4741, interpMode=RCIM_Cubic, leaveTangent=-1.4741, time=0.4906, value=0.4775}
         {arriveTangent=-2.6963, interpMode=RCIM_Cubic, leaveTangent=-2.6963, time=0.5283, value=0.7634}
         {arriveTangent=2.8896, interpMode=RCIM_Cubic, leaveTangent=2.8896, time=0.5517, value=0.3126}
         {arriveTangent=6.2844, interpMode=RCIM_Cubic, leaveTangent=6.2844, time=0.5824, value=0.9196}
         {arriveTangent=1.4786, interpMode=RCIM_Cubic, leaveTangent=1.4786, time=0.5976, value=0.6006}
         {arriveTangent=4.8152, interpMode=RCIM_Cubic, leaveTangent=4.8152, time=0.6363, value=0.9993}
         {arriveTangent=2.8106, interpMode=RCIM_Cubic, leaveTangent=2.8106, time=0.6464, value=0.8356}
         {arriveTangent=3.2181, interpMode=RCIM_Cubic, leaveTangent=3.2181, time=0.6964, value=1.1681}
         {arriveTangent=1.8672, interpMode=RCIM_Cubic, leaveTangent=1.8672, time=0.7177, value=1.065}
         {arriveTangent=-12.2147, interpMode=RCIM_Cubic, leaveTangent=-12.2147, time=0.7385, value=1.2469}
         {arriveTangent=-0.0928, interpMode=RCIM_Cubic, leaveTangent=-0.0928, tangentMode=RCTM_User, time=0.7564, value=0.5916}
         {arriveTangent=3.085, interpMode=RCIM_Cubic, leaveTangent=3.085, time=0.7931, value=1.02}
         {arriveTangent=2.1222, interpMode=RCIM_Cubic, leaveTangent=2.1222, time=0.8092, value=0.7545}
         {arriveTangent=-0.424, interpMode=RCIM_Cubic, leaveTangent=-0.424, time=0.8924, value=1.2306}
         {arriveTangent=-8.511, interpMode=RCIM_Cubic, leaveTangent=-8.511, time=0.911, value=0.7114}
         {arriveTangent=3.2422, interpMode=RCIM_Cubic, leaveTangent=3.2422, time=0.9278, value=0.9291}
         {interpMode=RCIM_Cubic, leaveTangent=0.0463, time=1, value=1}
        ]}}}
     CurveIndex=$Emitter.NormalizedLoopAge
     Scale Curve=10
    )
   )
  ]
  rendererClasses [
   &NiagaraMeshRendererProperties
   &NiagaraSpriteRendererProperties
   &NiagaraLightRendererProperties
  ]
  renderers [
   {rendererClass=&NiagaraMeshRendererProperties, rendererIndex=0}
   {rendererClass=&NiagaraSpriteRendererProperties, rendererIndex=1}
   {rendererClass=&NiagaraMeshRendererProperties, rendererIndex=2}
   {rendererClass=&NiagaraLightRendererProperties, rendererIndex=3}
  ]
 }
 SoulShards {
  emitterUpdateScript [
   &EmitterState()
   &SpawnParticlesFromOtherEmitter(
    Attribute Reader=DataInterface() {propertyValues {EmitterBinding {bindingMode=Other, emitterName=Soul}}}
    Spawn Rate=&Multiply_Float(A=15, B=$System.InterpolateOverTime.InterpolateOverTime_Float.Moving Average)
    Use Maximum Age=false
    Use Minimum Age=false
   )
  ]
  particleSpawnScript [
   &InitializeParticle(
    Lifetime Max=1
    Lifetime Min=0.7
    Lifetime Mode=::NewEnumerator1
    Mesh Scale Mode=::NewEnumerator1
    Mesh Uniform Scale Max=0.8
    Mesh Uniform Scale Min=0.35
    UsePositionOffset=false
    Write Material Random=true
   )
   &SampleParticlesFromOtherEmitter(Mesh Orientation Sampling:ENiagara_DirectReadSamplingMode=::NewEnumerator0)
   &CameraOffset(Camera Offset Amount=&RandomRangeFloat(Minimum=-8, Maximum=8))
   &InitialMeshOrientation(Use Rotation Vector=false)
  ]
  particleUpdateScript [
   &ParticleState()
   &ShapeLocation(
    Additional Quaternion Rotation=false
    Apply To Particle Position=::NewEnumerator1
    MaskSpawnGroup=false
    Non Uniform Scale=<0.8 1 2>
    Offset=<0 0 12>
    Offset Mode=::NewEnumerator0
    Override Seed=false
    Rotation Mode=::NewEnumerator3
    Rotation Quaternion=&MultiplyQuaternion(
     Quaternion A=&self:Get_MeshOrientation_Attribute_By_Index(
      Particle Attribute Reader:NiagaraDataInterfaceParticleRead=DataInterface() {propertyValues {EmitterBinding {bindingMode=Other, emitterName=Soul}}}
      Particle Index:NiagaraInt32=&MakeIntFromFloat(Float=$Particles.SampleParticlesFromOtherEmitter.SampledIndex)
     )
     Quaternion B=&FindQuatBetweenTwoVectors(From Vector=<1 0 0>, To Vector=<1 0 1>)
    )
    Sphere Radius=25
    Transform Order:ENiagara_TransformOrder=::NewEnumerator1
    UseOwnerScale=true
   )
   &InheritSourceMovement(
    Applied Position Scale=&VectorFromFloat(Value=&FloatFromCurve(FloatCurve=DataInterface() {propertyValues {Curve {keys [<at 0 1>,<at 1 0.7>]}}}))
    ApplyForce=false
    ApplyPosition=true
    ApplyPositionFalloff=false
   )
   &MeshRotationForce(Rotation=&RandomVector(Vector Scale=10))
   &SolveForcesAndVelocity(Clamp Velocity=false, Limit Acceleration=false)
   &DynamicMaterialParameters(
    Index 0 Param 1=0.02
    Index 0 Param 2=0.25
    Module.Param0WriteEnabled=true
    Module.Param1WriteEnabled=true
    Module.Param2WriteEnabled=false
    Module.Param3WriteEnabled=false
   )
   /SetVariables(Particles.Position=&Lerp_Position(
     A:NiagaraPosition=&Add_Vector(
      A=&self:Get_Position_Attribute_By_Index(
       Particle Attribute Reader:NiagaraDataInterfaceParticleRead=DataInterface() {propertyValues {EmitterBinding {bindingMode=Other, emitterName=Soul}}}
       Particle Index:NiagaraInt32=&MakeIntFromFloat(Float=$Particles.SampleParticlesFromOtherEmitter.SampledIndex)
      )
      B=$Particles.ShapeLocation.ShapeVector
     )
     B:NiagaraPosition=&AddVectorToPosition(
      Position=$Particles.Position
      Vector=&RotateVectorByQuaternion(
       Vector To Rotate=<0 0 10>
       Quaternion=&MultiplyQuaternion(
        Quaternion A=&self:Get_MeshOrientation_Attribute_By_Index(
         Particle Attribute Reader:NiagaraDataInterfaceParticleRead=DataInterface() {propertyValues {EmitterBinding {bindingMode=Other, emitterName=Soul}}}
         Particle Index:NiagaraInt32=&MakeIntFromFloat(Float=$Particles.SampleParticlesFromOtherEmitter.SampledIndex)
        )
        Quaternion B=&FindQuatBetweenTwoVectors(From Vector=<1 0 0>, To Vector=<1 0 1>)
       )
      )
     )
     Alpha:NiagaraFloat=&FloatFromCurve(FloatCurve=DataInterface() {propertyValues {Curve {keys [
          {arriveTangent=1.9048, interpMode=RCIM_Cubic, leaveTangent=1.9048, tangentMode=RCTM_User, time=0.1, value=0}
          {arriveTangent=0.5863, interpMode=RCIM_Cubic, leaveTangent=0.5863, tangentMode=RCTM_User, time=1, value=1}
         ]}}})
    )) {isSetParameters=true, module=SetVariables_#6}
   &ScaleMeshSize(Scale Factor=&VectorFromFloat(Value=&FloatFromCurve(
      FloatCurve=DataInterface() {propertyValues {Curve {keys [
          {arriveTangent=2.3595, interpMode=RCIM_Cubic, leaveTangent=2.3595, tangentMode=RCTM_User, time=0.2, value=0}
          <at 0.75 1 RCIM_Cubic>
          {arriveTangent=-4.7095, interpMode=RCIM_Cubic, leaveTangent=-4.7095, tangentMode=RCTM_User, time=1, value=0}
         ]}}}
      Scale Curve=$System.InterpolateOverTime.InterpolateOverTime_Float.Moving Average
     )))
  ]
  rendererClasses [&NiagaraMeshRendererProperties]
  renderers [{rendererClass=&NiagaraMeshRendererProperties, rendererIndex=0}]
 }
 SoulShards_FE {
  emitterUpdateScript [
   &EmitterState()
   &SpawnParticlesFromOtherEmitter(
    Attribute Reader=DataInterface() {propertyValues {EmitterBinding {bindingMode=Other, emitterName=Soul_FE}}}
    Spawn Rate=&Multiply_Float(A=15, B=$System.InterpolateOverTime.InterpolateOverTime_Float.Moving Average)
    Use Maximum Age=false
    Use Minimum Age=false
   )
  ]
  particleSpawnScript [
   &InitializeParticle(
    Lifetime Max=1
    Lifetime Min=0.7
    Lifetime Mode=::NewEnumerator1
    Mesh Scale Mode=::NewEnumerator1
    Mesh Uniform Scale Max=0.8
    Mesh Uniform Scale Min=0.35
    UsePositionOffset=false
    Write Material Random=true
   )
   &SampleParticlesFromOtherEmitter(Mesh Orientation Sampling:ENiagara_DirectReadSamplingMode=::NewEnumerator0)
   &CameraOffset(Camera Offset Amount=&RandomRangeFloat(Minimum=-8, Maximum=8))
   &InitialMeshOrientation(Use Rotation Vector=false)
  ]
  particleUpdateScript [
   &ParticleState()
   &ShapeLocation(
    Additional Quaternion Rotation=false
    Apply To Particle Position=::NewEnumerator1
    MaskSpawnGroup=false
    Non Uniform Scale=<0.8 1 2>
    Offset=<0 0 12>
    Offset Mode=::NewEnumerator0
    Override Seed=false
    Rotation Mode=::NewEnumerator3
    Rotation Quaternion=&MultiplyQuaternion(
     Quaternion A=&self:Get_MeshOrientation_Attribute_By_Index(
      Particle Attribute Reader:NiagaraDataInterfaceParticleRead=DataInterface() {propertyValues {EmitterBinding {bindingMode=Other, emitterName=Soul_FE}}}
      Particle Index:NiagaraInt32=&MakeIntFromFloat(Float=$Particles.SampleParticlesFromOtherEmitter.SampledIndex)
     )
     Quaternion B=&FindQuatBetweenTwoVectors(From Vector=<1 0 0>, To Vector=<1 0 1>)
    )
    Sphere Radius=25
    Transform Order:ENiagara_TransformOrder=::NewEnumerator1
    UseOwnerScale=true
   )
   &InheritSourceMovement(
    Applied Position Scale=&VectorFromFloat(Value=&FloatFromCurve(FloatCurve=DataInterface() {propertyValues {Curve {keys [<at 0 1>,<at 1 0.7>]}}}))
    ApplyForce=false
    ApplyPosition=true
    ApplyPositionFalloff=false
   )
   &MeshRotationForce(Rotation=&RandomVector(Vector Scale=10))
   &SolveForcesAndVelocity(Clamp Velocity=false, Limit Acceleration=false)
   &DynamicMaterialParameters(
    Index 0 Param 1=0.02
    Index 0 Param 2=0.25
    Module.Param0WriteEnabled=true
    Module.Param1WriteEnabled=true
    Module.Param2WriteEnabled=false
    Module.Param3WriteEnabled=false
   )
   /SetVariables(Particles.Position=&Lerp_Position(
     A:NiagaraPosition=&Add_Vector(
      A=&self:Get_Position_Attribute_By_Index(
       Particle Attribute Reader:NiagaraDataInterfaceParticleRead=DataInterface() {propertyValues {EmitterBinding {bindingMode=Other, emitterName=Soul_FE}}}
       Particle Index:NiagaraInt32=&MakeIntFromFloat(Float=$Particles.SampleParticlesFromOtherEmitter.SampledIndex)
      )
      B=$Particles.ShapeLocation.ShapeVector
     )
     B:NiagaraPosition=&AddVectorToPosition(
      Position=$Particles.Position
      Vector=&RotateVectorByQuaternion(
       Vector To Rotate=<0 0 75>
       Quaternion=&MultiplyQuaternion(
        Quaternion A=&self:Get_MeshOrientation_Attribute_By_Index(
         Particle Attribute Reader:NiagaraDataInterfaceParticleRead=DataInterface() {propertyValues {EmitterBinding {bindingMode=Other, emitterName=Soul_FE}}}
         Particle Index:NiagaraInt32=&MakeIntFromFloat(Float=$Particles.SampleParticlesFromOtherEmitter.SampledIndex)
        )
        Quaternion B=&FindQuatBetweenTwoVectors(From Vector=<1 0 0>, To Vector=<1 0 1>)
       )
      )
     )
     Alpha:NiagaraFloat=&FloatFromCurve(FloatCurve=DataInterface() {propertyValues {Curve {keys [
          {arriveTangent=1.9048, interpMode=RCIM_Cubic, leaveTangent=1.9048, tangentMode=RCTM_User, time=0.1, value=0}
          {arriveTangent=0.5863, interpMode=RCIM_Cubic, leaveTangent=0.5863, tangentMode=RCTM_User, time=1, value=1}
         ]}}})
    )) {isSetParameters=true, module=SetVariables_#6}
   &ScaleMeshSize(Scale Factor=&VectorFromFloat(Value=&FloatFromCurve(
      FloatCurve=DataInterface() {propertyValues {Curve {keys [
          {arriveTangent=2.3595, interpMode=RCIM_Cubic, leaveTangent=2.3595, tangentMode=RCTM_User, time=0.2, value=0}
          <at 0.75 1 RCIM_Cubic>
          {arriveTangent=-4.7095, interpMode=RCIM_Cubic, leaveTangent=-4.7095, tangentMode=RCTM_User, time=1, value=0}
         ]}}}
      Scale Curve=$System.InterpolateOverTime.InterpolateOverTime_Float.Moving Average
     )))
  ]
  rendererClasses [&NiagaraMeshRendererProperties]
  renderers [{rendererClass=&NiagaraMeshRendererProperties, rendererIndex=0}]
 }
 Soul_FE {
  emitterUpdateScript [
   &EmitterState()
   &SpawnBurst_Instantaneous(
    Spawn Count=&MakeIntFromFloat(Float=&FN_FrontEndFloatModifier(Gameplay=0, Front End=3))
    Use Spawn Probability=false
    UseLoopCountLimit=true
   )
   &InterpolateOverTime(Rate Of Change=50, Target Vector Value:Vector3f=$Engine.Owner.Velocity, Variable Type=::NewEnumerator2)
  ]
  particleSpawnScript [
   &FN_SetSpawnGroup(Number of Spawn Groups=3)
   &InitializeParticle(
    Lifetime=2
    Mesh Scale Mode=::NewEnumerator3
    Mesh Uniform Scale=$System.Ship Scale
    Sprite Size Mode=::NewEnumerator3
    Uniform Sprite Size=40
    UsePositionOffset=true
    Write Material Random=true
   )
  ]
  particleUpdateScript [
   &ParticleState(Kill Particles When Lifetime Has Elapsed=false, Let Infinitely Lived Particles Die When Emitter Deactivates=true)
   /SetVariables(Particles.Ambient:Vector3f=&Multiply_VectorByFloat(
     Vector=&MakeVector(
      X=&Waveform(
       "Waveform [1]"=::NewEnumerator12
       "[1] Input Position"=&RandomRangeFloat(Minimum=0, Maximum=1000)
       "[1] Frequency"=1
       "[1] Amplitude Min/Max"=<-50 50>
      )
      Y=&Waveform(
       "Waveform [1]"=::NewEnumerator12
       "[1] Input Position"=&RandomRangeFloat(Minimum=1200, Maximum=1800)
       "[1] Frequency"=1
       "[1] Amplitude Min/Max"=<-50 50>
      )
      Z=&Waveform(
       "Waveform [1]"=::NewEnumerator12
       "[1] Input Position"=&RandomRangeFloat(Minimum=2200, Maximum=2800)
       "[1] Frequency"=1
       "[1] Amplitude Min/Max"=<-50 50>
      )
     )
     Float=0.35
    )) {isSetParameters=true, module=SetVariables_#9}
   &InterpolateOverTime(
    Rate Of Change=3
    Target Float Value:NiagaraFloat=&MakeCustomFloatFromBool(Bool=$System.Timeline001.IsRewinding, True Float=0, False Float=1)
   ) {module=InterpolateOverTime004}
   /SetVariables(Particles.CatchupVector:Vector3f=&Multiply_VectorByFloat(
     Vector=&Multiply_VectorByFloat(
      Vector=&NormalizeVector(Vector To Normalize=$System.SkydiveVelocity, Safe Normalize=true, Safe Fallback Threshold=0, Safe Fallback Vector=<0 0 0.001>)
      Float=-75
     )
     Float=&Lerp_Float(
      A=&FloatFromCurve(
       FloatCurve=DataInterface() {propertyValues {Curve {keys [
           {interpMode=RCIM_Cubic, tangentMode=RCTM_User, time=0, value=0}
           {interpMode=RCIM_Cubic, tangentMode=RCTM_User, time=0.35, value=-0.35}
           {tangentMode=RCTM_User, time=1, value=0}
          ]}}}
       CurveIndex=$System.Timeline001.NormalizedTime
      )
      B=&FloatFromCurve(
       FloatCurve=DataInterface() {propertyValues {Curve {keys [
           {arriveTangent=3.3827, arriveTangentWeight=1.1106, interpMode=RCIM_Cubic, leaveTangentWeight=1.1106, time=0, value=0}
           {arriveTangentWeight=0.5241, interpMode=RCIM_Cubic, leaveTangentWeight=0.1667, time=0.5, value=0.65}
           <at 1 0 RCIM_Cubic>
          ]}}}
       CurveIndex=$System.Timeline001.NormalizedTime
      )
      Alpha=$Particles.InterpolateOverTime004.InterpolateOverTime_Float.Moving Average
     )
    )) {isSetParameters=true, module=SetVariables_#10}
   &ShapeLocation(
    Apply To Particle Position=::NewEnumerator1
    MaskSpawnGroup=false
    Override Seed=false
    Ring / Disc Distribution Mode:ENiagaraSphereDistributionMode=::NewEnumerator1
    Ring Radius=&Lerp_Float(A=100, B=125, Alpha=$System.InterpolateOverTime.InterpolateOverTime_Float.Moving Average)
    Rotation Angle=0.75
    Rotation Axis=<0 1 0>
    Rotation Coordinate Space=::World
    Rotation Mode=::NewEnumerator1
    Shape Primitive=::NewEnumerator4
    U Position=&ReturnNormalizedExecIndex(Normalized Index Includes One=false)
    UseOwnerScale=true
   )
   &SkeletalMeshLocation(
    Bone Sampling Mode=::NewEnumerator4
    MaskSpawnGroup=false
    Orientation Sampling=::NewEnumerator1
    Sampled Position Offset=&Add_Vector(
     A=$Particles.CatchupVector
     B=&RotateVector(
      Vector To Rotate:Vector3f=&Add_Vector(
       A=$Particles.Ambient
       B=&Add_Vector(
        A=&Add_Vector(
         A=$Particles.ShapeLocation.ShapeVector
         B=&MaskVectorBySpawnGroup(
          Masked Vector=&Lerp_Vector(A=<0 110 -50>, B=<0 110 -150>, Alpha=$System.InterpolateOverTime.InterpolateOverTime_Float.Moving Average)
          Passthrough Vector=<0 0 0>
         )
        )
        B=$System.OverallTargetOffsetFE
       )
      )
      Yaw:NiagaraFloat=&Waveform(
       "Waveform [1]"=::NewEnumerator12
       "[1] Input Position"=&RandomRangeFloat(Minimum=80000, Maximum=40000)
       "[1] Frequency"=1.25
       "[1] Amplitude Min/Max"=<7 -7>
      )
      Pitch:NiagaraFloat=0
      Roll:NiagaraFloat=0
      Rotation Rate:NiagaraFloat=2
     )
    )
    Skeletal Mesh=$User.SkeletalMesh
    Use Sampled Mesh Attribute as Ribbon ID=false
   )
   &ScaleMeshSize(Scale Factor=&VectorFromFloat(Value=&FloatFromCurve(
      FloatCurve=DataInterface() {propertyValues {Curve {keys [{arriveTangent=1.0186, interpMode=RCIM_Cubic, leaveTangent=1.0186, tangentMode=RCTM_User, time=0, value=0}, <at 1 1 RCIM_Cubic>]}}}
      CurveIndex=$Emitter.Age
     )))
   /SetVariables(Particles.Wobble:Quat4f=&MultiplyQuaternion(
     Quaternion A=&FindQuatBetweenTwoVectors(
      From Vector=&MakeVector(
       X=&Sine(
        Normalized Angle=&Add_Float(A=$System.FN_Integrator.Value, B=&RandomRangeFloat(Minimum=0, Maximum=1))
        Period=&MakeCustomFloatFromBool(Bool=&RandomBool(), True Float=1, False Float=-1)
        Scale=1
        Bias=0
       )
       Y=&Cosine(
        Normalized Angle=&Add_Float(A=$System.FN_Integrator.Value, B=&RandomRangeFloat(Minimum=0, Maximum=1))
        Period=&MakeCustomFloatFromBool(Bool=&RandomBool(), True Float=1, False Float=-1)
        Scale=1
        Bias=0
       )
       Z=&RandomRangeFloat(Minimum=6, Maximum=8)
      )
      To Vector=<0 0 1>
     )
     Quaternion B=$Particles.Initial.MeshOrientation
    )) {isSetParameters=true, module=SetVariables_#11}
   &UpdateMeshOrientation(
    Bank Rate=900000000
    Banking Angle Min / Max:Vector2f=<0.1 -0.1>
    "Foward Vector (Mesh Space)":Vector3f=&Lerp_Vector(A=<0.5 0 1>, B=<1 0 0>, Alpha=$System.InterpolateOverTime.InterpolateOverTime_Float.Moving Average)
    Orientation Method=::NewEnumerator3
    Particle Velocity:Vector3f=&Multiply_VectorByFloat(
     Vector=&Add_Vector(
      A=&ReturnMeshOrientationAxis(Mesh Orientation Axis=::YAxis, Mesh Orientation Quat=$Particles.SkeletalMeshLocation.SampledRotation)
      B=&Lerp_Vector(A=<0 0 -2>, B=<0 0 -0.5>, Alpha=$System.Dive)
     )
     Float=500000
    )
    Rotation Decay Rate In Degrees=900000000
    Up Vector:Vector3f=&Lerp_Vector(
     A=&ReturnMeshOrientationAxis(Mesh Orientation Axis=::YAxis, Mesh Orientation Quat=$Particles.SkeletalMeshLocation.SampledRotation)
     B=&Multiply_VectorByFloat(
      Vector=&ReturnMeshOrientationAxis(Mesh Orientation Axis=::ZAxis, Mesh Orientation Quat=$Particles.SkeletalMeshLocation.SampledRotation)
      Float=1
     )
     Alpha=&FloatFromCurve(FloatCurve=DataInterface() {propertyValues {Curve {keys [<at 0 0.5>,<at 0.25 0.75>]}}}, CurveIndex=$System.Dive)
    )
    Upright Roll Correction Speed=900000000
    Velocity Orientation Convergence Rate=900000000
   )
   &InterpolateOverTime(Rate Of Change=5, Target Quat Value:Quat4f=$Particles.MeshOrientation, Variable Type=::NewEnumerator5) {module=InterpolateOverTime002}
   /SetVariables(Particles.New Orient:Quat4f=$Particles.InterpolateOverTime002.InterpolateOverTime_Quaternion.Moving Average) {isSetParameters=true, module=SetVariables_#12}
   /SetVariables(Particles.MeshOrientation=&MultiplyQuaternion(Quaternion A=$Particles.Wobble, Quaternion B=$Particles.New Orient)) {isSetParameters=true, module=SetVariables_#13}
   &ScaleColor(
    Scale RGB=&MakeVectorFromLinearColorRGB(Color=&LerpLinearColors(StartColor=<rgba 0 1.3565 2 1>, EndColor=<rgba 4 0.479 0 1>, LerpFactor=$System.InterpolateOverTime.InterpolateOverTime_Float.Moving Average))
    ScaleA=true
    ScaleRGB=true
   )
   &DynamicMaterialParameters(
    Index 0 Param 1=1
    Index 0 Param 2=0.5
    Index 0 Param 3=&RandomRangeFloat(Minimum=0.1, Maximum=1)
    Index 0 Param 4=&RandomRangeFloat(Minimum=0.75, Maximum=1)
    Module.Param0WriteEnabled=true
    Module.Param1WriteEnabled=true
    Module.Param2WriteEnabled=true
    Module.Param3WriteEnabled=true
   )
   /SetVariables(
    Particles.Decal Color:LinearColor=&LerpLinearColors(
     StartColor=<rgba 0 1.3565 2 3>
     EndColor=<rgba 6 0.9039 0 3>
     LerpFactor=$System.InterpolateOverTime.InterpolateOverTime_Float.Moving Average
    )
    Particles.Decal Offset=15
    Particles.Decal Position:NiagaraPosition=&AddVectorToPosition(
     Position=$Particles.Position
     Vector=&RotateVectorByQuaternion(Vector To Rotate=<-10 0 15>, Quaternion=$Particles.MeshOrientation)
    )
   ) {isSetParameters=true, module=SetVariables_#14}
   &FN_Light_Attributes(
    Color=&LerpLinearColors(
     StartColor=<rgba 0 1.3565 2 1>
     EndColor=<rgba 6 0.9039 0 4>
     LerpFactor=$System.InterpolateOverTime.InterpolateOverTime_Float.Moving Average
    )
    Create Color Attribute=true
    Create Exponent Attribute=true
    Create Radius Attribute=true
    Exponent=2
    Radius=85
    Scale Color=&FloatFromCurve(
     FloatCurve=DataInterface() {propertyValues {Curve {keys [
         <at 0 1 RCIM_Cubic>
         {arriveTangent=1.794, interpMode=RCIM_Cubic, leaveTangent=1.794, time=0.0746, value=0.7914}
         {arriveTangent=-0.0227, interpMode=RCIM_Cubic, leaveTangent=-0.0227, tangentMode=RCTM_User, time=0.121, value=1.2171}
         {arriveTangent=-5.5976, interpMode=RCIM_Cubic, leaveTangent=-5.5976, time=0.1377, value=0.3589}
         {arriveTangent=5.9152, interpMode=RCIM_Cubic, leaveTangent=5.9152, time=0.1738, value=0.9213}
         {arriveTangent=3.6344, interpMode=RCIM_Cubic, leaveTangent=3.6344, time=0.2093, value=0.7825}
         {arriveTangent=-4.8832, interpMode=RCIM_Cubic, leaveTangent=-4.8832, time=0.2396, value=1.1602}
         {arriveTangent=-6.1936, interpMode=RCIM_Cubic, leaveTangent=-6.1936, time=0.254, value=0.5643}
         {arriveTangent=-3.4254, interpMode=RCIM_Cubic, leaveTangent=-3.4254, time=0.2791, value=0.9154}
         {arriveTangent=1.2156, interpMode=RCIM_Cubic, leaveTangent=1.2156, time=0.3019, value=0.4002}
         {arriveTangent=5.6358, interpMode=RCIM_Cubic, leaveTangent=5.6358, time=0.3406, value=0.9901}
         {arriveTangent=6.3907, interpMode=RCIM_Cubic, leaveTangent=6.3907, time=0.3796, value=0.8384}
         {arriveTangent=-3.0158, interpMode=RCIM_Cubic, leaveTangent=-3.0158, time=0.4112, value=1.4411}
         {arriveTangent=-10.4214, interpMode=RCIM_Cubic, leaveTangent=-10.4214, time=0.4339, value=0.6747}
         {arriveTangent=-3.4821, interpMode=RCIM_Cubic, leaveTangent=-3.4821, time=0.4676, value=0.8529}
         {arriveTangent=-1.4741, interpMode=RCIM_Cubic, leaveTangent=-1.4741, time=0.4906, value=0.4775}
         {arriveTangent=-2.6963, interpMode=RCIM_Cubic, leaveTangent=-2.6963, time=0.5283, value=0.7634}
         {arriveTangent=2.8896, interpMode=RCIM_Cubic, leaveTangent=2.8896, time=0.5517, value=0.3126}
         {arriveTangent=6.2844, interpMode=RCIM_Cubic, leaveTangent=6.2844, time=0.5824, value=0.9196}
         {arriveTangent=1.4786, interpMode=RCIM_Cubic, leaveTangent=1.4786, time=0.5976, value=0.6006}
         {arriveTangent=4.8152, interpMode=RCIM_Cubic, leaveTangent=4.8152, time=0.6363, value=0.9993}
         {arriveTangent=2.8106, interpMode=RCIM_Cubic, leaveTangent=2.8106, time=0.6464, value=0.8356}
         {arriveTangent=3.2181, interpMode=RCIM_Cubic, leaveTangent=3.2181, time=0.6964, value=1.1681}
         {arriveTangent=1.8672, interpMode=RCIM_Cubic, leaveTangent=1.8672, time=0.7177, value=1.065}
         {arriveTangent=-12.2147, interpMode=RCIM_Cubic, leaveTangent=-12.2147, time=0.7385, value=1.2469}
         {arriveTangent=-0.0928, interpMode=RCIM_Cubic, leaveTangent=-0.0928, tangentMode=RCTM_User, time=0.7564, value=0.5916}
         {arriveTangent=3.085, interpMode=RCIM_Cubic, leaveTangent=3.085, time=0.7931, value=1.02}
         {arriveTangent=2.1222, interpMode=RCIM_Cubic, leaveTangent=2.1222, time=0.8092, value=0.7545}
         {arriveTangent=-0.424, interpMode=RCIM_Cubic, leaveTangent=-0.424, time=0.8924, value=1.2306}
         {arriveTangent=-8.511, interpMode=RCIM_Cubic, leaveTangent=-8.511, time=0.911, value=0.7114}
         {arriveTangent=3.2422, interpMode=RCIM_Cubic, leaveTangent=3.2422, time=0.9278, value=0.9291}
         {interpMode=RCIM_Cubic, leaveTangent=0.0463, time=1, value=1}
        ]}}}
     CurveIndex=$Emitter.NormalizedLoopAge
     Scale Curve=10
    )
   )
  ]
  rendererClasses [
   &NiagaraMeshRendererProperties
   &NiagaraSpriteRendererProperties
   &NiagaraLightRendererProperties
  ]
  renderers [
   {rendererClass=&NiagaraMeshRendererProperties, rendererIndex=0}
   {rendererClass=&NiagaraSpriteRendererProperties, rendererIndex=1}
   {rendererClass=&NiagaraMeshRendererProperties, rendererIndex=2}
   {rendererClass=&NiagaraLightRendererProperties, rendererIndex=3}
  ]
 }
}
summary {
 systemName=NS_RedPepper_Contrail
 emitters [
  {
   emitterName=Soul
   rendererClasses [
    &NiagaraMeshRendererProperties
    &NiagaraSpriteRendererProperties
    &NiagaraLightRendererProperties
   ]
  }
  {emitterName=Fire, rendererClasses [&NiagaraSpriteRendererProperties]}
  {emitterName=Smoke, rendererClasses [&NiagaraSpriteRendererProperties]}
  {emitterName=FireAir, rendererClasses [&NiagaraSpriteRendererProperties]}
  {emitterName=SoulShards, rendererClasses [&NiagaraMeshRendererProperties]}
  {emitterName=Lightning, rendererClasses [&NiagaraMeshRendererProperties, &NiagaraLightRendererProperties]}
  {emitterName=Debris, rendererClasses [&NiagaraMeshRendererProperties]}
  {
   emitterName=Soul_FE
   rendererClasses [
    &NiagaraMeshRendererProperties
    &NiagaraSpriteRendererProperties
    &NiagaraLightRendererProperties
   ]
  }
  {emitterName=Fire_FE, rendererClasses [&NiagaraSpriteRendererProperties]}
  {emitterName=Smoke_FE, rendererClasses [&NiagaraSpriteRendererProperties]}
  {emitterName=FireAir_FE, rendererClasses [&NiagaraSpriteRendererProperties]}
  {emitterName=SoulShards_FE, rendererClasses [&NiagaraMeshRendererProperties]}
  {emitterName=Lightning_Mobile, rendererClasses [&NiagaraMeshRendererProperties]}
 ]
 userVariables [
  User.bIsFrontEnd {type=NiagaraBool}
  User.bIsFrontEndPreview {type=NiagaraBool}
  User.Color {type=LinearColor, description="", defaultValue {struct=LinearColor, value=<rgba 1 1 1 1>}}
  User.GlobalScaler {type=NiagaraFloat, defaultValue=0, description=""}
  User.ParaGlideLeanAlpha {type=NiagaraFloat, defaultValue=0, description=""}
  User.RibbonIDOffset {type=NiagaraInt32}
  User.SkeletalMesh {
   type=NiagaraDataInterfaceSkeletalMesh
   defaultValue {
    struct=NiagaraExt_VariableValue_DataInterface
    value {dataInterface=&self:NiagaraDataInterfaceSkeletalMesh_0}
   }
  }
 ]
}
